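; Base encounter component for a Sideways playspace, providing shared helper functions for encounters.
; SetRepParamBool, Float and Int send tagged values through the owning playspace's replicated param helper so UI can read encounter info, logging an error at each failed cast or validity check.
; EnableTrackingRadius registers a minimap radius circle around a target actor for all player controllers.
; RefreshActorsInVolume asks the owning playspace to refresh its Sideways actors; BeginPlay and Tick are empty.
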(event EventBeginPlay)

(event EventTick
  (params (DeltaSeconds :type Float%20%28single-precision%29)))

(fn EnableTrackingRadius
  (params (Radius :type Float%20%28double-precision%29) (CircleColor :type Linear%20Color%20Structure) (ActorTarget :type Actor%20Object%20Reference))
  ;[1 1 1] Add actor tracking info to show the bubbles radius. (note the 60s in the time between updates this is needed for now as the code doesn't handle a value of 0)
  (bind _returnvalue (RadiusTracker|MakeRadiusTrackingInitializer 0.0 Radius Radius 6e+01 0.0 CircleColor "/FortniteCoreUI/HUD/MiniMap/M_MiniMapCircle_Sideways.M_MiniMapCircle_Sideways" ActorTarget (Fortnite|GetAllFortPlayerControllers true) "Global"))
  (Fortnite|AddActorTrackingInfo (RadiusTracker|GetGameStateRadiusTrackerComponent) _returnvalue))

(fn RefreshActorsInVolume
  (bind _assideways_playspace (Utilities|Casting|CastToSidewaysPlayspace (Components|GetOwner)))
  (Class|SidewaysPlayspace|RefreshActorsinSideways _assideways_playspace))

(fn SetRepParamBool
  (params (ParamTag :type Gameplay%20Tag%20Structure) (Value :type Boolean))
  ;[1 1 1] Used to send Encounter info to a place where UI can access it.
  (bind _asfort_playspace (Utilities|Casting|CastToFortPlayspace (Components|GetOwner))
    (:then
      (Utilities|IsValid (Class|FortPlayspace|GetReplicatedState _asfort_playspace)
        (:Is%20Valid
          (Utilities|IsValid (Class|FortPlayspaceReplicatedState|GetReplicatedParamHelper (Class|FortPlayspace|GetReplicatedState _asfort_playspace))
            (:Is%20Valid
              (bind _returnvalue (Class|ActorComponentReplicatedParamHelper|SendParamasBool (Class|FortPlayspaceReplicatedState|GetReplicatedParamHelper (Class|FortPlayspace|GetReplicatedState _asfort_playspace)) ParamTag Value false))
              (return _returnvalue))
            (:Is%20Not%20Valid
              (Utilities|LogRepParamError "SetRepParamBool" "Param Helper"
                (:Out
                  _)))))
        (:Is%20Not%20Valid
          (Utilities|LogRepParamError "SetRepParamBool" "Replicated State"
            (:Out
              _))))
      (return false))
    (:CastFailed
      (Utilities|LogRepParamError "SetRepParamBool" "Playspace"
        (:Out
          _)))))

(fn SetRepParamFloat
  (params (ParamTag :type Gameplay%20Tag%20Structure) (Value :type Float%20%28double-precision%29))
  ;[1 1 1] Used to send Encounter info to a place where UI can access it.
  (bind _asfort_playspace (Utilities|Casting|CastToFortPlayspace (Components|GetOwner))
    (:then
      (Utilities|IsValid (Class|FortPlayspace|GetReplicatedState _asfort_playspace)
        (:Is%20Valid
          (Utilities|IsValid (Class|FortPlayspaceReplicatedState|GetReplicatedParamHelper (Class|FortPlayspace|GetReplicatedState _asfort_playspace))
            (:Is%20Valid
              (bind _returnvalue (Class|ActorComponentReplicatedParamHelper|SendParamasFloat (Class|FortPlayspaceReplicatedState|GetReplicatedParamHelper (Class|FortPlayspace|GetReplicatedState _asfort_playspace)) ParamTag Value false))
              (return _returnvalue))
            (:Is%20Not%20Valid
              (Utilities|LogRepParamError "SetRepParamFloat" "Param Helper"
                (:Out
                  _)))))
        (:Is%20Not%20Valid
          (Utilities|LogRepParamError "SetRepParamFloat" "Replicated State"
            (:Out
              _))))
      (return false))
    (:CastFailed
      (Utilities|LogRepParamError "SetRepParamFloat" "Playspace"
        (:Out
          _)))))

(fn SetRepParamInt
  (params (ParamTag :type Gameplay%20Tag%20Structure) (Value :type Integer))
  ;[1 1 1] Used to send Encounter info to a place where UI can access it.
  (bind _asfort_playspace (Utilities|Casting|CastToFortPlayspace (Components|GetOwner))
    (:then
      (Utilities|IsValid (Class|FortPlayspace|GetReplicatedState _asfort_playspace)
        (:Is%20Valid
          (Utilities|IsValid (Class|FortPlayspaceReplicatedState|GetReplicatedParamHelper (Class|FortPlayspace|GetReplicatedState _asfort_playspace))
            (:Is%20Valid
              (bind _returnvalue (Class|ActorComponentReplicatedParamHelper|SendParamasInt (Class|FortPlayspaceReplicatedState|GetReplicatedParamHelper (Class|FortPlayspace|GetReplicatedState _asfort_playspace)) ParamTag Value false))
              (return _returnvalue))
            (:Is%20Not%20Valid
              (Utilities|LogRepParamError "SetRepParamInt" "Param Helper"
                (:Out
                  _)))))
        (:Is%20Not%20Valid
          (Utilities|LogRepParamError "SetRepParamInt" "Replicated State"
            (:Out
              _))))
      (return false))
    (:CastFailed
      (Utilities|LogRepParamError "SetRepParamInt" "Playspace"
        (:Out
          _)))))
