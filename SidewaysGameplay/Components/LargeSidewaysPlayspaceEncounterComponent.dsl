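; Playspace component that runs the large Sideways encounter, spawning Deimos AI through a Living World Manager encounter and tagging entering players for quests.
; Server setup copies EnvCore references, feeds radius-keyed EQS params into the spawn volume, then starts a random encounter type matching the playlist team size.
; Tracked deaths pass a hotfixable threshold, pausing spawns until room exists, then advancing to the boss stage.
; Boss counts and POI-cleared state feed replicated UI params.
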(event Custom|OnPlayspaceUserAdded
  (params (AddedUser_UserId :type Unique%20Net%20Id%20Structure) (AddedUser_LeafPlayspace :type Playspace%20Object%20Reference) (AddedUser_PlayerStateCached :type Player%20State%20Object%20Reference) (AddedUser_ControllerCached :type Controller%20Object%20Reference))
  ;[1 1 1] Register to on user added event
  (CallFunction|StartEncounter)
  (bind _asfort_player_controller (Utilities|Casting|CastToFortPlayerController AddedUser_ControllerCached))
  (GameplayEffects|ApplyGameplayEffectToTarget :self (Ability|GetAbilitySystemComponent (Game|Player|GetPlayerPawn _asfort_player_controller)) :GameplayEffectClass "/SidewaysGameplay/Environmental/SidewaysEncounter/GE_Athena_Sideways_Quest_Visited.GE_Athena_Sideways_Quest_Visited_C" :Target (Ability|GetAbilitySystemComponent (Game|Player|GetPlayerPawn _asfort_player_controller)) :Level 0.0) ; Apply GE with tags for Quests to track participation
)

(event Custom|SetupSidewaysEncounterLogic
  (Utilities|FlowControl|Sequence
    (:then_0
      ;[1 1 1] Register to on user added event
      (bind _assideways_playspace (Utilities|Casting|CastToSidewaysPlayspace (Components|GetOwner)))
      (Playspace|BindEventtoNotifyPlayspaceUserAdded _assideways_playspace (delegate Custom|OnPlayspaceUserAdded))
      (Class|FortAsyncActionCurrentPlaylistReady|PlaylistDataReadyAsync
        (:OnReady
          (CallFunction|RefreshActorsinVolume))))
    (:then_1
      (if (Networking|IsServer)
        (bind _gamestate (Class|FortAsyncActionCurrentPlaylistReady|PlaylistDataReadyAsync)
          (:OnReady
            (Variables|Default|SetCachedGameState _gamestate)
            (CallFunction|CopyParentComponentReferencesandVariables :!pinned true) ; Initial setup
            (CallFunction|SetEQSParamsIntoEncounterPointProvider :!pinned true) ; Feed EQS Params into Spawner
            (CallFunction|StartLWMEncounter :!pinned true) ; Start LWM Encounter!
            (bind _encounterstagevolume (Variables|Default|GetEncounterStageVolume)
              (:then
                (bind _childactor (Variables|/Script/Engine.ChildActorComponent|GetChildActor _encounterstagevolume)
                  (:then
                    (bind _asfort_athena_living_world_volume (Utilities|Casting|CastToFortAthenaLivingWorldVolume _childactor))
                    (EventDispatchers|AssignOnActorSpawnedResult _asfort_athena_living_world_volume (delegate Custom|OnActorSpawnedResult_Event))
                    (CallFunction|UpdatePOIUI))
                  (:else)))
              (:else)))) ; Make sure we load any hotfix overrides
      ))))

(event EventBeginPlay
  (|CallParentFunction))

(event Custom|OnActorSpawnedResult_Event
  (params (SpawnedActor :type Actor%20Object%20Reference) (bResult :type Boolean))
  (bind _spawnedpawns (Variables|Default|GetSpawnedPawns))
  (if bResult
    (Utilities|FlowControl|Sequence
      (:then_0
        (bind _asfort_pawn (Utilities|Casting|CastToFortPawn SpawnedActor))
        ;[1 1 1] Track Eliminations for a Boss Wave
        (Utilities|Array|AddUnique (Variables|Default|GetSpawnedPawns) _asfort_pawn)
        (EventDispatchers|AssignOnDied _spawnedpawns (delegate Custom|OnDied_Event))
        (Game|AssignOnEndPlay _spawnedpawns (delegate Custom|OnEndPlay_Event))
        (if (GameplayTags|HasMatchingGameplayTag _asfort_pawn "(TagName=\"NPC.CharacterType.Brute.Mega\")")
          (Utilities|Array|AddUnique (Variables|Default|GetBossesSpawned) _asfort_pawn))
        (CallFunction|UpdatePOIUI))
      (:then_1
        ;[1 1 1] Set SpawnFX on Deimos
        (bind _asenemy_pawn_parent_deimos (Utilities|Casting|CastToEnemyPawn_Parent_Deimos SpawnedActor))
        (Variables|/Game/Characters/Enemies/Blueprints/EnemyPawn_Parent_Deimos.EnemyPawn_Parent_Deimos_C|SetGC_OnSpawnFX (Variables|Default|GetGC%20Spawn%20AI) _asenemy_pawn_parent_deimos)))))

(event Custom|StartEncounter
  ;[1 1 1] Start the Encounter
  (if (not (Variables|Default|GetbIsEncounterActive))
    (Variables|Default|SetbIsEncounterActive true)))

(event Custom|OnEndPlay_Event
  (params (Actor :type Actor%20Object%20Reference) (EndPlayReason :type EEndPlayReason%20Enum :default "Destroyed"))
  ;[1 1 1] Track Eliminations for a Boss Wave
  (bind _asfort_pawn (Utilities|Casting|CastToFortPawn Actor))
  (bind _data (Utilities|FlowControl|MergeExecutionwithData _asfort_pawn (stale Custom|OnDied_Event _asfort_pawn))
    (:Then
      (EventDispatchers|UnbindEventfromOnDied _data (delegate Custom|OnDied_Event))
      (Game|UnbindEventfromOnEndPlay _data (delegate Custom|OnEndPlay_Event))
      (Utilities|Array|RemoveItem (Variables|Default|GetSpawnedPawns) _data)
      (if (Utilities|Array|ContainsItem (Variables|Default|GetBossesSpawned) _data)
        (Math|Integer|IncrementInt (Variables|Default|GetBossesEliminated)
          (:%20%20
            (Utilities|Array|RemoveItem (Variables|Default|GetBossesSpawned) _data))))
      (CallFunction|UpdatePOIUI))))

(event Custom|OnDied_Event
  (params (DamagedActor :type Actor%20Object%20Reference) (Damage :type Float%20%28single-precision%29) (InstigatedBy :type Controller%20Object%20Reference) (DamageCauser :type Actor%20Object%20Reference) (HitLocation :type Vector) (FHitComponent :type Primitive%20Component%20Object%20Reference) (BoneName :type Name) (Momentum :type Vector))
  (bind _bossstageenabled (Variables|Default|GetBossStageEnabled))
  (bind _returnvalue (not _bossstageenabled))
  (bind _returnvalue_2 (Game|GetGameState))
  (bind (_asbool _asintrounded) (Default|GetScalableFloatHotfixValue (Variables|Default|GetBossRoundEliminationThreshold_HF)))
  (bind (_asbool_2 _asintrounded_2) (Default|GetScalableFloatHotfixValue (Variables|Default|GetMaxNumAI_HF)))
  (bind (_asbool_3 _asintrounded_3) (Default|GetScalableFloatHotfixValue (Variables|Default|GetNumberOfBossesWeAreSpawning_HF)))
  ;[1 1 1] Track Eliminations for a Boss Wave
  (bind _asfort_pawn (Utilities|Casting|CastToFortPawn DamagedActor))
  (Utilities|FlowControl|Sequence
    (:then_0
      (if (Utilities|NotEqual(Object) DamagedActor DamageCauser)
        (bind _result (Math|Integer|IncrementInt (Variables|Default|GetAIEliminated))
          (:%20%20
            (if (and (>= _result _asintrounded) _returnvalue)
              (if (>= (- _asintrounded_2 (Utilities|Array|Length (Variables|Default|GetSpawnedPawns))) _asintrounded_3)
                (bind _lwmencounterinstance (Variables|Default|GetLWMEncounterInstance)
                  (:then
                    (Class|FortAthenaLivingWorldEncounterInstance|AdvanceEncounterStage _lwmencounterinstance :!pinned true) ; Advance to BOSS Stage
                    (Class|FortAthenaLivingWorldEncounterInstance|ResumeEncounter _lwmencounterinstance)
                    (Class|FortAthenaLivingWorldEncounterInstance|RequestEventGeneration _lwmencounterinstance) ; Replenish those empty spaces with Bosses NOW
                    (Variables|Default|SetBossStageEnabled true))
                  (:else))
                (else
                  (Utilities|IsValid _returnvalue_2
                    (:Is%20Valid
                      (Utilities|IsValid (Actor|GetComponentbyClass _returnvalue_2 "/Script/LagerRuntime.FortAthenaLivingWorldManager")
                        (:Is%20Valid
                          (bind _lwmencounterinstance_2 (Variables|Default|GetLWMEncounterInstance)
                            (:then
                              (Class|FortAthenaLivingWorldEncounterInstance|PauseEncounter _lwmencounterinstance_2) ; Pause Encounter so further Elims can make space for Bosses to spawn together.
                            )
                            (:else))))))))
              (elif _returnvalue
                _))))))
    (:then_1
      (bind _data (Utilities|FlowControl|MergeExecutionwithData (stale Custom|OnEndPlay_Event _asfort_pawn) _asfort_pawn)
        (:Then
          (EventDispatchers|UnbindEventfromOnDied _data (delegate Custom|OnDied_Event))
          (Game|UnbindEventfromOnEndPlay _data (delegate Custom|OnEndPlay_Event))
          (Utilities|Array|RemoveItem (Variables|Default|GetSpawnedPawns) _data)
          (if (Utilities|Array|ContainsItem (Variables|Default|GetBossesSpawned) _data)
            (Math|Integer|IncrementInt (Variables|Default|GetBossesEliminated)
              (:%20%20
                (Utilities|Array|RemoveItem (Variables|Default|GetBossesSpawned) _data))))
          (CallFunction|UpdatePOIUI))))))

(fn CopyParentComponentReferencesAndVariables
  (bind _buildingowner (Default|GetBuildingOwner))
  (Utilities|IsValid _buildingowner
    (:Is%20Valid
      (bind _encounterstagevolume (Variables|/SidewaysGameplay/Environmental/SidewaysEncounter/BGA_Athena_SidewaysEncounter_EnvCoreLarge.BGA_Athena_SidewaysEncounter_EnvCoreLarge_C|GetEncounterStageVolume _buildingowner)
        (:then
          (Variables|Default|SetEncounterStageVolume _encounterstagevolume)
          (bind _encounterbubble (Variables|/SidewaysGameplay/Environmental/SidewaysEncounter/BGA_Athena_SidewaysEncounter_EnvCoreLarge.BGA_Athena_SidewaysEncounter_EnvCoreLarge_C|GetEncounterBubble (Default|GetBuildingOwner))
            (:then
              (Variables|Default|SetEncounterBubble _encounterbubble)
              (Variables|Default|SetEncounterBubbleVisualRadius (Variables|/SidewaysGameplay/Environmental/SidewaysEncounter/BGA_Athena_SidewaysEncounter_EnvCoreLarge.BGA_Athena_SidewaysEncounter_EnvCoreLarge_C|GetVisualRadiusInSteppedIncrements (Default|GetBuildingOwner))))
            (:else)))
        (:else)))))

(fn GetBuildingOwner
  (bind _testactor (LocalVariables|GetTestActor))
  (bind _returnvalue (Actor|GetParentActor _testactor))
  (bind _testactor_2 (LocalVariables|GetTestActor))
  (bind _asfort_playspace (Utilities|Casting|CastToFortPlayspace (Components|GetOwner)))
  (LocalVariables|SetTestActor (Variables|/Script/PlayspaceSystem.Playspace|GetBoundGameplayVolume _asfort_playspace)) ; First set our owner as TestActor
  (Utilities|IsValid _testactor_2
    (:Is%20Valid
      (bind _asbga_athena_sideways_encounter_env_core_large (Utilities|Casting|CastToBGA_Athena_SidewaysEncounter_EnvCoreLarge _testactor_2)
        (:then
          _)
        (:CastFailed
          (Utilities|IsValid _returnvalue
            (:Is%20Valid
              (LocalVariables|SetTestActor _returnvalue) ; If it has a parent actor, set TestActor and try again
            )
            (:Is%20Not%20Valid
              _)) ; If not, check if the parent actor is valid
          (return 0))) ; Check if the test actor is an EnvCore
      (return _asbga_athena_sideways_encounter_env_core_large))
    (:Is%20Not%20Valid
      _)))

(fn GetPlaylistGametype
  (bind _cachedgamestate (Variables|Default|GetCachedGameState)
    (:then
      (return (GameplayTags|HasTag (Class|FortGameStateAthena|GetAthenaPlaylistContextTags _cachedgamestate) "(TagName=\"Athena.Playlist.Solo\")") (GameplayTags|HasTag (Class|FortGameStateAthena|GetAthenaPlaylistContextTags _cachedgamestate) "(TagName=\"Athena.Playlist.Duo\")") (GameplayTags|HasTag (Class|FortGameStateAthena|GetAthenaPlaylistContextTags _cachedgamestate) "(TagName=\"Athena.Playlist.Trios\")") (or (GameplayTags|HasTag (Class|FortGameStateAthena|GetAthenaPlaylistContextTags _cachedgamestate) "(TagName=\"Athena.Playlist.Squad\")") (GameplayTags|HasTag (Class|FortGameStateAthena|GetAthenaPlaylistContextTags _cachedgamestate) "(TagName=\"Athena.Playlist.TeamRumble\")"))))
    (:else)))

(fn GetScalableFloatHotfixValue
  (params (ScalableFloat :type Scalable%20Float%20Structure%20%28by%20ref%29) (AtLevel :type Float%20%28double-precision%29))
  (if (Default|ScalableFloatIsValid ScalableFloat)
    (return (> (Math|Float|Round (ScalableFloat|GetValueatLevel ScalableFloat AtLevel)) 0) (Math|Float|Round (ScalableFloat|GetValueatLevel ScalableFloat AtLevel)) (ScalableFloat|GetValueatLevel ScalableFloat AtLevel))
    (else
      (return (> (Math|Float|Round (Utilities|Struct|BreakScalableFloat ScalableFloat)) 0) (Math|Float|Round (Utilities|Struct|BreakScalableFloat ScalableFloat)) (Utilities|Struct|BreakScalableFloat ScalableFloat)))))

(fn GetSidewaysBubble
  (bind _testactor (LocalVariables|GetTestActor))
  (bind _returnvalue (Actor|GetParentActor _testactor))
  (bind _testactor_2 (LocalVariables|GetTestActor))
  (bind _asfort_playspace (Utilities|Casting|CastToFortPlayspace (Components|GetOwner)))
  (LocalVariables|SetTestActor (Variables|/Script/PlayspaceSystem.Playspace|GetBoundGameplayVolume _asfort_playspace)) ; First set our owner as TestActor
  (Utilities|IsValid _testactor_2
    (:Is%20Valid
      (bind _asbga_athena_sideways_encounter_bubble_large (Utilities|Casting|CastToBGA_Athena_SidewaysEncounter_BubbleLarge _testactor_2)
        (:then
          _)
        (:CastFailed
          (Utilities|IsValid _returnvalue
            (:Is%20Valid
              (LocalVariables|SetTestActor _returnvalue) ; If it has a parent actor, set TestActor and try again
            )
            (:Is%20Not%20Valid
              _)) ; If not, check if the parent actor is valid
          (return 0))) ; Check if the test actor is a Bubble
      (return _asbga_athena_sideways_encounter_bubble_large))
    (:Is%20Not%20Valid
      _)))

(fn ScalableFloatIsValid
  (params (ScalableFloat :type Scalable%20Float%20Structure%20%28by%20ref%29))
  (bind (_value _curve) (Utilities|Struct|BreakScalableFloat ScalableFloat))
  (bind (_curvetable _rowname) (Utilities|Struct|BreakCurveTableRowHandle _curve))
  (return (and (Utilities|IsValid _curvetable :!class /Script/Engine.KismetSystemLibrary) (Utilities|Name|NotEqual(Name) _rowname "None"))))

(fn SetEQSParamsIntoEncounterPointProvider
  (bind _encounterbubblevisualradius (Variables|Default|GetEncounterBubbleVisualRadius))
  (bind _returnvalue (Math|Float|%25%28Float%29 _encounterbubblevisualradius 1e+03))
  (bind _returnvalue_2 (/ _returnvalue 1e+03))
  (bind _returnvalue_3 (Math|Float|Round _returnvalue_2))
  (bind _returnvalue_4 (* _returnvalue_3 1e+03))
  (bind _returnvalue_5 (- _encounterbubblevisualradius _returnvalue))
  (bind _returnvalue_6 (+ _returnvalue_5 _returnvalue_4))
  (bind _encountereqsdistancedata (Variables|Default|GetEncounterEQSDistanceData))
  (bind _value (Utilities|Map|Find _encountereqsdistancedata _returnvalue_6))
  (bind _eqsparamvalues_7_09cd566c464068c815e5af98c9ec22dc (Utilities|Struct|BreakStructAthenaSidewaysEncounterLEQSData _value))
  (bind _keys (Utilities|Map|Keys _eqsparamvalues_7_09cd566c464068c815e5af98c9ec22dc)) ; Each Key in this Struct is an EQS Param 
  (for _array_element _keys
    (bind _encounterstagevolume (Variables|Default|GetEncounterStageVolume)
      (:then
        (bind _childactor (Variables|/Script/Engine.ChildActorComponent|GetChildActor _encounterstagevolume)
          (:then
            (bind _asfort_athena_living_world_volume (Utilities|Casting|CastToFortAthenaLivingWorldVolume _childactor))
            (Class|FortAthenaLivingWorldVolume|SetEQSQueryConfigParam _asfort_athena_living_world_volume _array_element (Utilities|Map|Find _eqsparamvalues_7_09cd566c464068c815e5af98c9ec22dc _array_element)) ; Feed Params into the Point Provider Volume that match the Visual Radius we are going for.
            (Development|PrintString (Utilities|String|BuildString(Float) (Utilities|String|BuildString(Name) "SETTING " "PARAM " _array_element " ") "VAL = " (Utilities|Map|Find _eqsparamvalues_7_09cd566c464068c815e5af98c9ec22dc _array_element))))
          (:else)))
      (:else))))

(fn StartLWMEncounter
  (bind (_solo _duo _trio _squad_or_rumble) (CallFunction|GetPlaylistGametype))
  (bind _returnvalue (Game|GetGameState))
  (bind _returnvalue_2 (Actor|GetComponentbyClass _returnvalue "/Script/LagerRuntime.FortAthenaLivingWorldManager"))
  (Utilities|IsValid _returnvalue
    (:Is%20Valid
      (Utilities|IsValid _returnvalue_2
        (:Is%20Valid
          (bind _encounterstagevolume (Variables|Default|GetEncounterStageVolume)
            (:then
              (bind _childactor (Variables|/Script/Engine.ChildActorComponent|GetChildActor _encounterstagevolume)
                (:then
                  (if (.bSuccess (Utilities|Casting|CastToFortAthenaLivingWorldPointProviderInterface _childactor))
                    (Class|FortAthenaLivingWorldPointProviderInterface|EnablePointProvider_DEPRECATED (Utilities|Casting|CastToFortAthenaLivingWorldPointProviderInterface _childactor))
                    (bind _returnvalue_3 (Class|FortAthenaLivingWorldManager|StartEncounter _returnvalue_2 (Utilities|Array|RandomArrayItem (select _solo (Variables|Default|GetDeimosEncounterTypes_Solo) (select _duo (Variables|Default|GetDeimosEncounterTypes_Duo) (select _trio (Variables|Default|GetDeimosEncounterTypes_Trio) (select _squad_or_rumble (Variables|Default|GetDeimosEncounterTypes_Squad) (Variables|Default|GetDeimosEncounterTypes_Solo)))))) (Utilities|Array|MakeArray (Utilities|Casting|CastToFortAthenaLivingWorldPointProviderInterface _childactor)) (Default|GetBuildingOwner) true :!pinned true)) ; Start Encounter!
                    (Variables|Default|SetLWMEncounterInstance _returnvalue_3)))
                (:else)))
            (:else)))))))

(fn UpdatePOIUI
  (bind _numberofbosseswearespawning_hf (Variables|Default|GetNumberOfBossesWeAreSpawning_HF))
  (bind (_asbool _asintrounded) (Default|GetScalableFloatHotfixValue _numberofbosseswearespawning_hf))
  (bind _bosseseliminated (Variables|Default|GetBossesEliminated))
  (Default|SetRepParamInt "(TagName=\"RepParam.Integer.Sideways.Encounter.Bosses.Max\")" _asintrounded) ; Used for UI.
  (Default|SetRepParamInt "(TagName=\"RepParam.Integer.Sideways.Encounter.Bosses.Remaining\")" (Utilities|Array|Length (Variables|Default|GetBossesSpawned))) ; Used for UI.
  (Default|SetRepParamInt "(TagName=\"RepParam.Integer.Sideways.Encounter.Bosses.Eliminated\")" _bosseseliminated) ; Used for UI.
  (Default|SetRepParamBool "(TagName=\"RepParam.Bool.Sideways.POICleared\")" (>= _bosseseliminated _asintrounded)) ; Used for UI.
)
