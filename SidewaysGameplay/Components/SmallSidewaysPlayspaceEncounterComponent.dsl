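; Component that runs a multi-stage, score-based AI encounter for Fortnite's small Sideways bubble.
; Each stage enables a Living World volume, tracks spawned and eliminated pawn scores, and carries surplus points forward.
; Function-name timers drive min/max waves, countdown music, no-spawn failure and last-stage cleanup, deciding progression or closure.
; Overkill scores and timer durations replicate to UI through RepParams, while prototype EQS logic places chests and pillars.
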(event Custom|EncounterTimerCountdown
  ;[1 1 1] Countdown Timer Music
  (CallFunction|UpdateTimerUI) ; When Countdown Music Starts
  (if (or (not (or (Default|EncounterStageScoreEliminationTargetHit) (Default|EncounterStageScoreEliminationFallbackTargetHit))) (Math|Boolean|ANDBoolean (or (Default|EncounterStageScoreEliminationTargetHit) (Default|EncounterStageScoreEliminationFallbackTargetHit)) (Default|EncounterHasSpawnedAI) (Utilities|Enum|Equal(Enum) (CallFunction|GetEncounterState) "NewEnumerator5")))
    (Default|UpdateLoopingAudio (Variables|Music|GetMusicCountdown)) ; Switch to Countdown Music
  ))

(event Custom|TrackOverkillTotal
  ;[1 1 1] Track Overkill Score Points so we can payout a Bonus at the end
  (bind _output_get (Variables|EncounterLogic|SetEncounterTotalOverkillPoints (+ (Variables|EncounterLogic|GetEncounterTotalOverkillPoints) (Variables|EncounterLogic|GetEncounterStageOverkillPoints))))
  (Default|DebugPrintString (Utilities|String|BuildString(Integer) "" "TOTAL OVERKILL POINTS = " _output_get)))

(event Custom|SetCleanupTimeout
  (bind _encounterlaststagecleanuptimeoutduration_hf (Variables|EncounterSetup|GetEncounterLastStageCleanupTimeoutDuration_HF))
  (bind (_asbool _asintrounded _asfloat) (Default|GetScalableFloatHotfixValue _encounterlaststagecleanuptimeoutduration_hf))
  (bind _bubble (CallFunction|GetSidewaysBubble))
  (bind _lwmencounterinstance (Variables|EncounterLogic|GetLWMEncounterInstance))
  ;[1 1 1] Max Cleanup Timer after EncounterTimer elapses, you've won, but you still have AI to clean up
  (if (Math|Boolean|ANDBoolean (not (Variables|EncounterLogic|GetEncounterNoSpawnTimeoutFailure)) (Utilities|Enum|Equal(Enum) (CallFunction|GetEncounterState) "NewEnumerator5") (not (Utilities|Time|IsTimerActivebyFunctionName 0 "CleanupTimeout")) (or (Default|EncounterStageScoreEliminationTargetHit) (Default|EncounterStageScoreEliminationFallbackTargetHit)))
    (Class|BGAAthenaSidewaysEncounterBubbleSmall|EncounterStageCompleted _bubble)
    (Class|BGAAthenaSidewaysEncounterBubbleSmall|EncounterBeaten _bubble)
    (if (> _asfloat 0.0)
      (Default|SetRepParamFloat "(TagName=\"RepParam.Float.Sideways.CleanupTimeoutDuration.Max\")" _asfloat) ; Used for UI.
      (Utilities|Time|SetTimerbyFunctionName 0 "CleanupTimeout" _asfloat)
      (CallFunction|EncounterTimerCountdown) ; Kick-Off Countdown Music
      (Default|CallSpawnCosmeticEvent (Variables|Spawn/Despawn/DestroyCosmetics|GetGC_Stage3Success))
      (if (Variables|EncounterSetup|GetAutoKillEnemiesBeforeClosingBubble)
        (CallFunction|ForciblyAutoKillEnemiesBeforeClosingBubble)
        (Utilities|IsValid _lwmencounterinstance
          (:Is%20Valid
            (Class|FortAthenaLivingWorldEncounterInstance|PauseEncounter _lwmencounterinstance)))))))

(event Custom|SetEncounterTimer
  (bind _encounterstagewarningcountdownduration_hf (Variables|EncounterSetup|GetEncounterStageWarningCountdownDuration_HF))
  (bind (_asbool _asintrounded _asfloat) (Default|GetScalableFloatHotfixValue _encounterstagewarningcountdownduration_hf))
  (bind _currentencounterwavedurationmax (Variables|EncounterLogic|GetCurrentEncounterWaveDurationMax))
  (bind _returnvalue (- _currentencounterwavedurationmax _asfloat))
  (bind _currentencounterwavedurationmax_2 (Variables|EncounterLogic|GetCurrentEncounterWaveDurationMax))
  (bind _currentencounterwavedurationmin (Variables|EncounterLogic|GetCurrentEncounterWaveDurationMin))
  ;[1 1 1] Start Encounter Timer
  (Utilities|SwitchHasAuthority
    (:Authority
      (Utilities|FlowControl|Sequence
        (:then_0
          (if (and (not (Utilities|Time|IsTimerActivebyFunctionName 0 "EncounterTimer")) (Variables|EncounterLogic|GetEncounterVolumeEnabled))
            (Default|SetRepParamFloat "(TagName=\"RepParam.Float.Sideways.WaveDuration.Max\")" _currentencounterwavedurationmax_2) ; Used for UI.
            (Utilities|Time|SetTimerbyFunctionName 0 "EncounterTimer" _currentencounterwavedurationmax_2)
            (Variables|EncounterLogic|SetEncounterTimerEnabled true)
            (Utilities|Time|ClearTimerbyFunctionName 0 "EncounterNoSpawnTimeout") ; Clear NoSpawn Timeout since we are now spawning.
            (Default|SetRepParamFloat "(TagName=\"RepParam.Float.Sideways.WarningCountdownDuration.Max\")" _asfloat) ; Used for UI.
            (Utilities|Time|SetTimerbyFunctionName 0 "EncounterTimerCountdown" _returnvalue) ; Start intense music shortly before the timer runs out.
            (Default|DebugPrintString (Utilities|String|Append (Utilities|String|BuildString(Float) (Utilities|String|BuildString(Float) (Utilities|String|BuildString(Float) "STARTING ENCOUNTER | " "TIMER MAX = " (Variables|EncounterLogic|GetCurrentEncounterWaveDurationMax) " | ") "TIMER MIN = " (Variables|EncounterLogic|GetCurrentEncounterWaveDurationMin) " | ") "COUNTDOWN TIMER = " _returnvalue " | ") (Utilities|String|EnumtoString (CallFunction|GetEncounterState))))
            (Default|SetRepParamFloat "(TagName=\"RepParam.Float.Sideways.WaveDuration.Min\")" _currentencounterwavedurationmin) ; Used for UI.
            (if (> _currentencounterwavedurationmin 0.0)
              (Utilities|Time|SetTimerbyFunctionName 0 "EncounterTimerMin" _currentencounterwavedurationmin))))
        (:then_1
          (CallFunction|UpdateTimerUI) ; AI Spawned - After we set any timers (if any), update with our current remaining duration(s)
        )))))

(event Custom|CleanupTimeout
  ;[1 1 1] Max Cleanup Timer after EncounterTimer elapses, you've won, but you still have AI to clean up
  (CallFunction|UpdateTimerUI) ; When we run out of Cleanup Time
  (if (Variables|EncounterSetup|GetPlayersMustCleanupEnemies)
    (Variables|EncounterLogic|SetEncounterCleanupTimeoutFailure true)
    (if (Utilities|Enum|Equal(Enum) (CallFunction|GetEncounterState) "NewEnumerator5")
      (Default|UpdateLoopingAudio (Variables|Music|GetMusicLoss)) ; Switch to Loser Music
    )
    (elif (or (Default|EncounterStageScoreEliminationTargetHit) (Default|EncounterStageScoreEliminationFallbackTargetHit))
      (Default|SetEncounterSuccessVariables)))
  (Default|ProgressState))

(event Custom|SetEncounterNoSpawnTimeout
  (bind (_asbool _asintrounded _asfloat) (Default|GetScalableFloatHotfixValue (Variables|EncounterSetup|GetEncounterNoSpawnTimeoutDuration_HF)))
  ;[1 1 1] No Spawn Timeout - If we try to spawn, but don't, timeout and fail
  (Utilities|SwitchHasAuthority
    (:Authority
      (if (and (not (Default|EncounterHasSpawnedAI)) (not (Utilities|Time|IsTimerActivebyFunctionName 0 "EncounterNoSpawnTimeout")))
        (Utilities|Time|SetTimerbyFunctionName 0 "EncounterNoSpawnTimeout" _asfloat)))))

(event Custom|EncounterNoSpawnTimeout
  ;[1 1 1] No Spawn Timeout - If we try to spawn, but don't, timeout and fail
  (Variables|EncounterLogic|SetEncounterNoSpawnTimeoutFailure true)
  (CallFunction|DisableVolumeandTimerThenProgressState))

(event Custom|StartEncounterStage
  (bind (_encounterwavedurationmax _encounterwavedurationmin _encounterpointeliminationtarget _encounterloottier _overkillloottiers) (Default|GetEncounterStageData))
  (bind (_asbool _asintrounded _asfloat) (Default|GetScalableFloatHotfixValue _encounterwavedurationmax))
  (bind (_asbool_2 _asintrounded_2) (Default|GetScalableFloatHotfixValue _encounterpointeliminationtarget))
  (bind (_asbool_3 _asintrounded_3 _asfloat_2) (Default|GetScalableFloatHotfixValue _encounterwavedurationmin))
  ;[1 1 1] Kick-Off next Encounter Stage
  (Utilities|SwitchHasAuthority
    (:Authority
      (Utilities|FlowControl|Sequence
        (:then_1
          ;[1 1 1] Reset Encounter Variables
          (Variables|EncounterLogic|SetEncounterStageScoreSpawned 0)
          (if (> (Variables|EncounterLogic|GetEncounterStageScoreEliminated) 0)
            (Variables|EncounterLogic|SetEncounterStageScoreEliminated (- (Variables|EncounterLogic|GetEncounterStageScoreEliminated) (Variables|EncounterLogic|GetEncounterStageScoreTarget))) ; Carry-over the previous over-the-threshold points so we count them for the next stage
            (else
              (Variables|EncounterLogic|SetEncounterStageScoreEliminated 0)))
          (Variables|EncounterLogic|SetEncounterStageOverkillPoints 0)
          (bind _output_get (Variables|EncounterLogic|SetOverkillEnabled false))
          (Default|SetRepParamBool "(TagName=\"RepParam.Bool.Sideways.OverkillEnabled\")" _output_get) ; Used for UI.
        )
        (:then_2
          ;[1 1 1] Get correct Encounter Stage Volume
          (switch Utilities|FlowControl|Switch|SwitchonEnum_Athena_SidewaysEncounter_State (CallFunction|GetEncounterState)
            (:NewEnumerator1
              _)
            (:NewEnumerator2
              _)
            (:NewEnumerator5
              _))
          (bind _encounterstagevolume (Variables|Default|GetEncounterStageVolume)
            (:then
              (bind _childactor (Variables|/Script/Engine.ChildActorComponent|GetChildActor _encounterstagevolume)
                (:then
                  (bind _asfort_athena_living_world_volume (Utilities|Casting|CastToFortAthenaLivingWorldVolume _childactor))
                  (Variables|EncounterLogic|SetCurrentLivingWorldEncounterPointProviderVolume _asfort_athena_living_world_volume))
                (:else)))
            (:else)))
        (:then_3
          ;[1 1 1] Gather Encounter Stage Info and set Variables
          (Variables|EncounterLogic|SetCurrentEncounterWaveDurationMax _asfloat)
          (Variables|EncounterLogic|SetCurrentEncounterWaveDurationMin _asfloat_2)
          (bind _output_get_2 (Variables|EncounterLogic|SetEncounterStageScoreTarget _asintrounded_2))
          (if (>= (Variables|EncounterLogic|GetEncounterStageScoreEliminated) _output_get_2)
            (Default|DebugPrintString "Carried-over more points than our new Target. Shouldn't happen."))
          (Variables|EncounterLogic|SetCurrentEncounterLootTier _encounterloottier)
          (CallFunction|SetEQSParamsIntoEncounterPointProvider)
          (bind _keys (Utilities|Map|Keys _overkillloottiers))
          (bind _output_get_3 (Variables|EncounterLogic|SetOverkillThresholdL (Utilities|Array|Get(acopy) _keys)))
          (bind _output_get_4 (Variables|EncounterLogic|SetOverkillThresholdM (Utilities|Array|Get(acopy) _keys 1)))
          (bind _output_get_5 (Variables|EncounterLogic|SetOverkillThresholdS (Utilities|Array|Get(acopy) _keys 2)))
          (Default|SetRepParamInt "(TagName=\"RepParam.Integer.Sideways.Encounter.Total.Overkill.Target.L\")" _output_get_3) ; Used for UI.
          (Default|SetRepParamInt "(TagName=\"RepParam.Integer.Sideways.Encounter.Total.Overkill.Target.M\")" _output_get_4) ; Used for UI.
          (Default|SetRepParamInt "(TagName=\"RepParam.Integer.Sideways.Encounter.Total.Overkill.Target.S\")" _output_get_5) ; Used for UI.
        )
        (:then_4
          ;[1 1 1] Enable Encounter Volume - Start Spawning - Start Timers
          (CallFunction|EnableandBindVolume)
          (Variables|EncounterLogic|SetEncounterStageActive true) ; Track that we're NOT in a transitional phase.\nBubble scaling is done.
          (CallFunction|RequestSpawnEventGeneration) ; Try to kickstart and fulfill spawning for this volume ASAP.
          (CallFunction|SetEncounterNoSpawnTimeout) ; Set fallback timer in case no AI spawn.
        )))))

(event Custom|TryToProgressEncounter
  (bind _returnvalue (Default|EncounterTimersasString))
  ;[1 1 1] Progress to the next Encounter or State if Encounter is Disabled, and AI are Dead.
  (if (Variables|EncounterLogic|GetEncounterNoSpawnTimeoutFailure)
    (CallFunction|TrytoCloseEncounter) ; FAILED
    (elif (Variables|EncounterLogic|GetEncounterStageActive)
      (if (Variables|EncounterLogic|GetEncounterVolumeEnabled)
        (if (Default|EncounterStageScoreEliminationTargetHit)
          (if (Default|EncounterStageScoreEliminationTargetOverkill)
            (Default|DebugPrintString (Utilities|String|Append "Encounter Active - ELIM THRESHOLD HIT - MIN TIMER NOT HIT - O V E R K I L L ! | " (Default|EncounterTimersasString)))
            (elif (Variables|EncounterLogic|GetEncounterTimerEnabled)
              (Default|DebugPrintString (Utilities|String|Append "Encounter Active - ELIM THRESHOLD AND MIN TIMER HIT - Success! | " (Default|EncounterTimersasString)))
              (CallFunction|DisableVolumeandTimerThenProgressState) ; Make sure our volumes are off, timers are disabled, and we start the next stage of the encounter.
            ))
          (elif (not (Default|EncounterHasSpawnedAI))
            (CallFunction|RequestSpawnEventGeneration) ; Try to kickstart and fulfill spawning for this volume ASAP.
          ))
        (elif (Utilities|Enum|Equal(Enum) (CallFunction|GetEncounterState) "NewEnumerator5")
          (if (or (not (Default|EncounterHasSpawnedAI)) (and (not (Variables|EncounterSetup|GetPlayersMustCleanupEnemies)) (not (Utilities|Time|IsTimerActivebyFunctionName 0 "CleanupTimeout"))))
            _)))
      (if (Default|EncounterStageScoreEliminationTargetHit)
        _
        (elif (Default|EncounterStageScoreEliminationFallbackTargetHit)
          _
          (else
            (CallFunction|TrytoCloseEncounter) ; FAILED
          )))
      (if (Variables|EncounterLogic|GetEncounterCleanupTimeoutFailure)
        (Default|DebugPrintString (Utilities|String|Append "Encounter Done - But last wave cleanup was too slow. Semi fail | " _returnvalue))
        (else
          (Default|SetEncounterSuccessVariables)
          (Default|DebugPrintString (Utilities|String|Append "Encounter Done - Success! | " _returnvalue))))
      (Default|ProgressState))))

(event Custom|EncounterSoloA
  ;[1 1 1] Debug Encounters
  (Variables|EncounterSetup|SetDeimosEncounterForcedTest "/Lager/Lager/Encounters/Deimos/Sideways/LagerEncounter_Deimos_Sideways_Solo_A.LagerEncounter_Deimos_Sideways_Solo_A"))

(event Custom|EncounterSoloB
  ;[1 1 1] Debug Encounters
  (Variables|EncounterSetup|SetDeimosEncounterForcedTest "/Lager/Lager/Encounters/Deimos/Sideways/LagerEncounter_Deimos_Sideways_Solo_B.LagerEncounter_Deimos_Sideways_Solo_B"))

(event Custom|EncounterSoloC
  ;[1 1 1] Debug Encounters
  (Variables|EncounterSetup|SetDeimosEncounterForcedTest "/Lager/Lager/Encounters/Deimos/Sideways/LagerEncounter_Deimos_Sideways_Solo_C.LagerEncounter_Deimos_Sideways_Solo_C"))

(event Custom|EncounterSoloE
  ;[1 1 1] Debug Encounters
  (Variables|EncounterSetup|SetDeimosEncounterForcedTest "/Lager/Lager/Encounters/Deimos/Sideways/LagerEncounter_Deimos_Sideways_Solo_E_MB.LagerEncounter_Deimos_Sideways_Solo_E_MB"))

(event Custom|FindActorsToSpawn
  (bind _encounterstate (CallFunction|GetEncounterState))
  (bind _map_encounterstatesettings (Variables|EncounterSetup|GetMAP_EncounterStateSettings))
  (bind (_value _returnvalue) (Utilities|Map|Find _map_encounterstatesettings _encounterstate))
  (bind (_encounterbubblevisualradius_10_3cf034c245f6cb6cad1939a9a2fb5b32 _encounterloopingaudio_8_3aa8ed3f4879a23432761f90e5473930 _environmentaleqs_28_62f841504fc73c2ab73ecdab064bddb4 _encounterspawnedenvironmentals_32_0f5e79b24e80bc980420f7acf7a83661) (Utilities|Struct|BreakStructAthenaSidewaysEncounterState _value))
  (bind (_chest_34_122da6de47ab0aee5bb057877b9adc45 _chests_minamount_22_d6ef526c45d747a71d2e678718b924df _chests_maxamount_23_348b992d449b29eb33888fb784ce4041 _chanceforsuperchest_10_9fa7abbb4d7f6e87780c01a7c0a49487 _superchest_33_63c10a4946349f3d9f5e03b9b1ce7abb _pillars_32_cc551fd049599d6fa20ac08077b5e2b8 _pillars_minamount_29_bd70636a4f868ec0a01ea289494c79e4 _pillars_maxamount_30_343259ed47b13e9f8af8cfab693721af _explosivepods_36_59506de44b8798490e51908f5eae5065) (Utilities|Struct|BreakStructAthenaSidewaysEncounterEnvActors _encounterspawnedenvironmentals_32_0f5e79b24e80bc980420f7acf7a83661))
  ;[1 0.7687 0] PROTO Work for selecting and spawning in actors. Most likely will be moved to code for better performance.
  (if _returnvalue
    (Default|AddEnvironmentalActors _chests_minamount_22_d6ef526c45d747a71d2e678718b924df _chests_maxamount_23_348b992d449b29eb33888fb784ce4041 _chest_34_122da6de47ab0aee5bb057877b9adc45 (Utilities|Array|MakeArray 0))
    (Default|AddEnvironmentalActors _pillars_minamount_29_bd70636a4f868ec0a01ea289494c79e4 _pillars_maxamount_30_343259ed47b13e9f8af8cfab693721af _explosivepods_36_59506de44b8798490e51908f5eae5065 _pillars_32_cc551fd049599d6fa20ac08077b5e2b8 true)
    (Utilities|Array|Shuffle (Variables|EncounterEnvironmentalActors|GetEnvironmentalActorsToSpawn))))

(event Custom|EncounterTimerMin
  ;[1 1 1] Min Encounter Timer Ends - See if we hit Elim Threshold and should Progress Encounter
  (CallFunction|UpdateTimerUI) ; When the Min Encounter Timer ends
  (Default|DebugPrintString (Utilities|String|Append "MIN ENCOUNTER TIMER COMPLETE | " (Default|EncounterTimersasString)))
  (Utilities|Time|ClearTimerbyFunctionName 0 "EncounterTimerMin")
  (CallFunction|TrytoProgressEncounter) ; See if ALL THE THINGS are already dead.
)

(event Custom|EncounterTimer
  ;[1 1 1] Encounter Timer Ends - Disable Encounter Volume - Try to End or Progress Encounter
  (Default|DebugPrintString (Utilities|String|Append "MAX ENCOUNTER TIMER COMPLETE | " (Utilities|String|EnumtoString (CallFunction|GetEncounterState))))
  (CallFunction|SetCleanupTimeout)
  (CallFunction|DisableandUnbindVolume)
  (if (not (or (Default|EncounterStageScoreEliminationTargetHit) (Default|EncounterStageScoreEliminationFallbackTargetHit)))
    (Default|UpdateLoopingAudio (Variables|Music|GetMusicLoss)) ; Switch to Loser Music
  )
  (CallFunction|TrytoProgressEncounter) ; See if ALL THE THINGS are already dead.
  (CallFunction|UpdateTimerUI) ; After Wave Timer or Elimination Threshold Hit
)

(event Custom|%20PlaceBuildingEnvironmentals
  (bind _encounterstate (CallFunction|GetEncounterState))
  (bind _map_encounterstatesettings (Variables|EncounterSetup|GetMAP_EncounterStateSettings))
  (bind (_value _returnvalue) (Utilities|Map|Find _map_encounterstatesettings _encounterstate))
  (bind (_encounterbubblevisualradius_10_3cf034c245f6cb6cad1939a9a2fb5b32 _encounterloopingaudio_8_3aa8ed3f4879a23432761f90e5473930 _environmentaleqs_28_62f841504fc73c2ab73ecdab064bddb4) (Utilities|Struct|BreakStructAthenaSidewaysEncounterState _value))
  (bind _buildingowner (CallFunction|GetBuildingOwner))
  ;[1 0.7687 0] PROTO Work for selecting and spawning in actors. Most likely will be moved to code for better performance.
  (if (Networking|IsServer)
    (if (and (Variables|EncounterSetup|GetSpawnEnvironmentActors) (Utilities|Enum|Equal(Enum) (CallFunction|GetEncounterState) "NewEnumerator1"))
      (if _returnvalue
        (Utilities|IsValid _environmentaleqs_28_62f841504fc73c2ab73ecdab064bddb4
          (:Is%20Valid
            (Utilities|IsValid _buildingowner
              (:Is%20Valid
                (bind _returnvalue_2 (AI|EQS|RunEQSQuery _environmentaleqs_28_62f841504fc73c2ab73ecdab064bddb4 _buildingowner "AllMatching"))
                (AI|EQS|BindEventtoOnQueryFinished _returnvalue_2 (delegate Custom|EQSForEnvironmentalsFinished))
                (CallFunction|FindActorstoSpawn)))))))))

(event Custom|CloseEncounter
  (bind _lwmencounterinstance (Variables|EncounterLogic|GetLWMEncounterInstance))
  ;[1 1 1] DEBUG - Force close Encounter (will kill living AI)
  (CallFunction|DisableandUnbindVolume)
  ;[1 1 1] FAILED - Try to Close Encounter
  (Utilities|Time|ClearTimerbyFunctionName 0 "EncounterTimer")
  (Utilities|Time|ClearTimerbyFunctionName 0 "EncounterTimerCountdown")
  (Variables|EncounterLogic|SetEncounterTimerEnabled false)
  (Utilities|Time|ClearTimerbyFunctionName 0 "EncounterTimerMin")
  (Utilities|Time|ClearTimerbyFunctionName 0 "EncounterNoSpawnTimeout")
  (Utilities|Time|ClearTimerbyFunctionName 0 "CleanupTimeout")
  (CallFunction|SetEncounterState "NewEnumerator3")
  (if (Variables|EncounterSetup|GetAutoKillEnemiesBeforeClosingBubble)
    (CallFunction|ForciblyAutoKillEnemiesBeforeClosingBubble)
    (Utilities|IsValid _lwmencounterinstance
      (:Is%20Valid
        (Class|FortAthenaLivingWorldEncounterInstance|PauseEncounter _lwmencounterinstance)))))

(event Custom|DisableVolumeAndTimerThenProgressState
  ;[1 1 1] Players have hit Elimination Target - Clear current Stage and TRY to Progress to Next Stage.
  (Utilities|Time|ClearTimerbyFunctionName 0 "EncounterTimer")
  (Utilities|Time|ClearTimerbyFunctionName 0 "EncounterTimerCountdown")
  (Variables|EncounterLogic|SetEncounterTimerEnabled false)
  (Utilities|Time|ClearTimerbyFunctionName 0 "EncounterTimerMin")
  (Utilities|Time|ClearTimerbyFunctionName 0 "CleanupTimeout")
  ;[1 1 1] Encounter Timer Ends - Disable Encounter Volume - Try to End or Progress Encounter
  (CallFunction|SetCleanupTimeout)
  (CallFunction|DisableandUnbindVolume)
  (if (not (or (Default|EncounterStageScoreEliminationTargetHit) (Default|EncounterStageScoreEliminationFallbackTargetHit)))
    (Default|UpdateLoopingAudio (Variables|Music|GetMusicLoss)) ; Switch to Loser Music
  )
  (CallFunction|TrytoProgressEncounter) ; See if ALL THE THINGS are already dead.
  (CallFunction|UpdateTimerUI) ; After Wave Timer or Elimination Threshold Hit
)

(event Custom|EQSForEnvironmentalsFinished
  (params (QueryInstance :type EQS%20Query%20Instance%20Object%20Reference) (QueryStatus :type EEnvQueryStatus%20Enum :default "Processing"))
  (bind _returnvalue (Math|Random|RandomFloatinRange 0.0 3.6e+02))
  (bind _row_pillarscalemax (Variables|EncounterEnvironmentalActors|GetRow_PillarScaleMax))
  (bind _returnvalue_2 (ScalableFloat|GetValueatLevel _row_pillarscalemax))
  (bind _row_pillarscalemin (Variables|EncounterEnvironmentalActors|GetRow_PillarScaleMin))
  (bind _returnvalue_3 (ScalableFloat|GetValueatLevel _row_pillarscalemin))
  (bind _returnvalue_4 (Math|Random|RandomFloatinRange _returnvalue_3 _returnvalue_2))
  (bind _returnvalue_5 (Math|Conversions|ToVector(Float) _returnvalue_4))
  (bind _buildingowner (CallFunction|GetBuildingOwner))
  (bind _buildingowner_2 (CallFunction|GetBuildingOwner))
  (bind _buildingowner_3 (CallFunction|GetBuildingOwner))
  (bind _environmentalactorstospawn (Variables|EncounterEnvironmentalActors|GetEnvironmentalActorsToSpawn))
  (bind _encounterchestscale (Variables|EncounterEnvironmentalActors|GetEncounterChestScale))
  (bind _pillarparentclass (Variables|EncounterEnvironmentalActors|GetPillarParentClass))
  (bind _environmentalszoffset (Variables|EncounterEnvironmentalActors|GetEnvironmentalsZOffset))
  ;[1 0.7687 0] PROTO Work for selecting and spawning in actors. Most likely will be moved to code for better performance.
  (bind _resultlocations (AI|EQS|GetQueryResultsasLocations QueryInstance))
  (bind _output_get (Variables|EncounterEnvironmentalActors|SetEnvironmentalActorLocations _resultlocations))
  (Utilities|Array|Shuffle _output_get)
  (bind (_array_element _array_index) (Utilities|Array|ForEachLoopwithBreak _resultlocations :!label _label)
    (:LoopBody
      (if (>= (Utilities|Array|Length _environmentalactorstospawn) (+ _array_index 1))
        (if (Utilities|ClassIsChildOf (Utilities|Array|Get(acopy) _environmentalactorstospawn _array_index) _pillarparentclass)
          (bind _actorspawned (CallFunction|SpawnPillar (Math|Transform|MakeTransform (Math|Vector|vector+vector _array_element (Math|Vector|MakeVector 0.0 0.0 _environmentalszoffset)) (Math|Rotator|MakeRotator 0.0 0.0 _returnvalue) (select (Utilities|ClassIsChildOf (Utilities|Array|Get(acopy) _environmentalactorstospawn _array_index) _pillarparentclass) _returnvalue_5 _encounterchestscale)) _buildingowner_2))
          (Utilities|IsValid _actorspawned
            (:Is%20Valid
              (Utilities|IsValid _buildingowner
                (:Is%20Valid
                  (Building|InitializeKismetSpawnedBuildingActor _actorspawned _buildingowner 0 true 0 false)
                  (Utilities|Array|AddUnique (Variables|EncounterEnvironmentalActors|GetSpawnedEnvironmentalActors) _actorspawned)))))
          (else
            (bind _actorspawned_2 (CallFunction|SpawnChest (Math|Transform|MakeTransform (Math|Vector|vector+vector _array_element (Math|Vector|MakeVector 0.0 0.0 _environmentalszoffset)) (Math|Rotator|MakeRotator 0.0 0.0 _returnvalue) (select (Utilities|ClassIsChildOf (Utilities|Array|Get(acopy) _environmentalactorstospawn _array_index) _pillarparentclass) _returnvalue_5 _encounterchestscale)) _buildingowner_2))
            (Utilities|IsValid _actorspawned_2
              (:Is%20Valid
                (Utilities|IsValid _buildingowner_3
                  (:Is%20Valid
                    (Building|InitializeKismetSpawnedBuildingActor _actorspawned_2 _buildingowner_3 0 true 0 false)
                    (Utilities|Array|AddUnique (Variables|EncounterEnvironmentalActors|GetSpawnedEnvironmentalActors) _actorspawned_2)))))))
        (else
          (goto _label Break))))))

(event Custom|RequestSpawnEventGeneration
  (bind _returnvalue (Game|GetGameState))
  (bind _returnvalue_2 (Actor|GetComponentbyClass _returnvalue "/Script/LagerRuntime.FortAthenaLivingWorldManager"))
  ;[1 1 1] Push LWM to create Spawn Events NOW.
  (Utilities|IsValid _returnvalue
    (:Is%20Valid
      (Utilities|IsValid _returnvalue_2
        (:Is%20Valid
          (Class|FortAthenaLivingWorldManager|RequestEventGeneration _returnvalue_2))))))

(event Custom|TryToCloseEncounter
  (bind _lwmencounterinstance (Variables|EncounterLogic|GetLWMEncounterInstance))
  ;[1 1 1] FAILED - Try to Close Encounter
  (if (or (not (Default|EncounterHasSpawnedAI)) (not (Variables|EncounterSetup|GetPlayersMustCleanupEnemies)))
    (if (Networking|IsServer)
      (Utilities|Time|ClearTimerbyFunctionName 0 "EncounterTimer")
      (Utilities|Time|ClearTimerbyFunctionName 0 "EncounterTimerCountdown")
      (Variables|EncounterLogic|SetEncounterTimerEnabled false)
      (Utilities|Time|ClearTimerbyFunctionName 0 "EncounterTimerMin")
      (Utilities|Time|ClearTimerbyFunctionName 0 "EncounterNoSpawnTimeout")
      (Utilities|Time|ClearTimerbyFunctionName 0 "CleanupTimeout")
      (CallFunction|SetEncounterState "NewEnumerator3")
      (if (Variables|EncounterSetup|GetAutoKillEnemiesBeforeClosingBubble)
        (CallFunction|ForciblyAutoKillEnemiesBeforeClosingBubble)
        (Utilities|IsValid _lwmencounterinstance
          (:Is%20Valid
            (Class|FortAthenaLivingWorldEncounterInstance|PauseEncounter _lwmencounterinstance)))))))

(event Custom|CleanUpEnvironmentals
  ;[1 0.7687 0] PROTO Work for selecting and spawning in actors. Most likely will be moved to code for better performance.
  (for _array_element (Variables|EncounterEnvironmentalActors|GetSpawnedEnvironmentalActors)
    (Utilities|IsValid _array_element
      (:Is%20Valid
        (bind _asprop_sideways_geo_rock_parent (Utilities|Casting|CastToProp_SidewaysGeo_Rock_Parent _array_element)
          (:then
            (Class|PropSidewaysGeoRockParent|DestroyDynamicPillar _asprop_sideways_geo_rock_parent))
          (:CastFailed
            (bind _asb_chest_athena_sideways_chest_super (Utilities|Casting|CastToB_Chest_Athena_SidewaysChest_Super _array_element))
            (Class|BChestAthenaSidewaysChestSuper|DestroyDynamicChest _asb_chest_athena_sideways_chest_super)))))))

(event Custom|EnableAndBindVolume
  ;[1 1 1] Enable/Disable Spawn Volumes
  (if (Variables|EncounterSetup|GetCanSpawnEnemies)
    (bind _currentlivingworldencounterpointprovidervolume (Variables|EncounterLogic|GetCurrentLivingWorldEncounterPointProviderVolume)
      (:then
        (EventDispatchers|AssignOnActorSpawnedResult _currentlivingworldencounterpointprovidervolume (delegate Custom|OnActorSpawnedResult_Event) :!class /Script/LagerRuntime.FortAthenaLivingWorldVolume)
        (Class|FortAthenaLivingWorldPointProviderInterface|EnablePointProvider_DEPRECATED _currentlivingworldencounterpointprovidervolume)
        (Variables|EncounterLogic|SetEncounterVolumeEnabled true)
        (if (Default|EncounterHasSpawnedAI)
          (CallFunction|SetEncounterTimer) ; IF we have AI from a previous Stage, start the Timer immediately.
        ))
      (:else))))

(event Custom|DisableAndUnbindVolume
  ;[1 1 1] Enable/Disable Spawn Volumes
  (if (Variables|EncounterSetup|GetCanSpawnEnemies)
    (bind _currentlivingworldencounterpointprovidervolume (Variables|EncounterLogic|GetCurrentLivingWorldEncounterPointProviderVolume)
      (:then
        (Class|FortAthenaLivingWorldVolume|DisablePointProvider _currentlivingworldencounterpointprovidervolume)
        (EventDispatchers|UnbindEventfromOnActorSpawnedResult _currentlivingworldencounterpointprovidervolume (delegate Custom|OnActorSpawnedResult_Event) :!class /Script/LagerRuntime.FortAthenaLivingWorldVolume)
        (Variables|EncounterLogic|SetEncounterVolumeEnabled false))
      (:else))))

(event Custom|OnEndPlay_Event
  (params (Actor :type Actor%20Object%20Reference) (EndPlayReason :type EEndPlayReason%20Enum :default "Destroyed"))
  ;[1 1 1] AI Died - Remove from List - Try to End or Progress Encounter
  (bind _asfort_pawn (Utilities|Casting|CastToFortPawn Actor))
  (bind _data (Utilities|FlowControl|MergeExecutionwithData _asfort_pawn (stale Custom|OnDied_Event _asfort_pawn))
    (:Then
      (EventDispatchers|UnbindEventfromOnDied _data (delegate Custom|OnDied_Event) :!class /Script/FortniteGame.FortPawn)
      (Game|UnbindEventfromOnEndPlay _data (delegate Custom|OnEndPlay_Event))
      (if (not (Variables|EncounterLogic|GetAutoKillInProgress))
        (Utilities|Array|RemoveItem (Variables|EncounterLogic|GetSpawnedPawns) _data))
      (CallFunction|TrytoProgressEncounter) ; See if ALL THE THINGS are dead and whether our Encounter is done.
      (CallFunction|ElimFeedback _data)
      (CallFunction|UpdateTimerUI) ; After an AI is Eliminated
    )))

(event Custom|ElimFeedback
  (params (EliminatedFortPawn :type Fort%20Pawn%20Object%20Reference))
  ;[1 1 1] Elim Feedback and Overkill Logic
  ; WIP
  (Default|CallElimFeedbackEvent EliminatedFortPawn)
  (if (Default|EncounterStageScoreEliminationTargetOverkill)
    (if (Variables|EncounterLogic|GetOverkillEnabled)
      (Default|CallSpawnCosmeticEvent (Variables|Spawn/Despawn/DestroyCosmetics|GetGC_Overkill_Normal)) ; Overkill Success Sound
      (else
        (bind _output_get (Variables|EncounterLogic|SetOverkillEnabled true)) ; Maybe OnRep this?
        (Default|SetRepParamBool "(TagName=\"RepParam.Bool.Sideways.OverkillEnabled\")" _output_get) ; Used for UI.
        (Utilities|Time|ClearTimerbyFunctionName 0 "EncounterTimerCountdown") ; If we are in Overkill, the countdown music should definitely not play
        (Default|UpdateLoopingAudio (Variables|Music|GetMusicOverkill)) ; Switch to Winner Music
        (Default|CallSpawnCosmeticEvent (Variables|Spawn/Despawn/DestroyCosmetics|GetGC_Overkill_First)) ; First Overkill
      ))
    (elif (or (Utilities|Time|IsTimerActivebyFunctionName 0 "EncounterTimer") (Utilities|Time|IsTimerActivebyFunctionName 0 "CleanupTimeout"))
      (Default|CallSpawnCosmeticEvent (Variables|Spawn/Despawn/DestroyCosmetics|GetGC_Elimimination_Positive)) ; Minor Success Sound
      (else
        (Default|CallSpawnCosmeticEvent (Variables|Spawn/Despawn/DestroyCosmetics|GetGC_Elimination_Neutral)) ; Neutral/Fail Sound
      ))))

(event Custom|OnDied_Event
  (params (DamagedActor :type Actor%20Object%20Reference) (Damage :type Float%20%28single-precision%29) (InstigatedBy :type Controller%20Object%20Reference) (DamageCauser :type Actor%20Object%20Reference) (HitLocation :type Vector) (FHitComponent :type Primitive%20Component%20Object%20Reference) (BoneName :type Name) (Momentum :type Vector))
  (bind (_overkill _overkillorunderkillscore) (Default|EncounterStageScoreEliminationTargetOverkill))
  (bind (_asbool _asintrounded) (Default|GetScalableFloatHotfixValue (Variables|EncounterSetup|GetNPC%20IO%20Score%20HF)))
  (bind (_asbool_2 _asintrounded_2) (Default|GetScalableFloatHotfixValue (Variables|EncounterSetup|GetNPC%20Misc%20Score%20HF)))
  (bind (_asbool_3 _asintrounded_3) (Default|GetScalableFloatHotfixValue (Variables|EncounterSetup|GetNPC%20SmallBag%20Score%20HF)))
  (bind _scorevalue (Variables|Default|GetScoreValue))
  ;[1 1 1] AI Died - Remove from List - Try to End or Progress Encounter
  (Default|DebugPrintString (Utilities|String|BuildString(Object) (Utilities|String|BuildString(Object) "" "DEATH INSTIG = " InstigatedBy " | ") "DEATH CAUSER = " DamageCauser))
  (bind _asfort_pawn (Utilities|Casting|CastToFortPawn DamagedActor))
  (Utilities|FlowControl|Sequence
    (:then_0
      (if (Utilities|NotEqual(Object) DamagedActor DamageCauser)
        (if (Variables|EncounterLogic|GetEncounterVolumeEnabled)
          (bind _asfort_aipawn (Utilities|Casting|CastToFortAIPawn _asfort_pawn)
            (:then
              (Variables|Default|SetScoreValue (FortAIPawn|GetScoreValue _asfort_aipawn)))
            (:CastFailed
              (Utilities|IsValid _asfort_pawn
                (:Is%20Valid
                  (bind _asfort_player_pawn_athena (Utilities|Casting|CastToFortPlayerPawnAthena _asfort_pawn))
                  (Variables|Default|SetScoreValue (Math|Integer|SelectInt _asintrounded (Math|Integer|SelectInt _asintrounded_3 _asintrounded_2 (GameplayTags|HasMatchingGameplayTag _asfort_player_pawn_athena "(TagName=\"AISpawnerData.Type.Boss.SmallBag\")")) (GameplayTags|HasMatchingGameplayTag _asfort_player_pawn_athena "(TagName=\"AISpawnerData.Type.IO\")")))))))
          (Variables|EncounterLogic|SetEncounterTotalScoreEliminated (+ _scorevalue (Variables|EncounterLogic|GetEncounterTotalScoreEliminated))) ; Track number of AI Points we have Eliminated cumulatively during the Encounter.
          (bind _output_get (Variables|EncounterLogic|SetEncounterStageScoreEliminated (+ _scorevalue (Variables|EncounterLogic|GetEncounterStageScoreEliminated)))) ; Track number of AI Points we have Eliminated cumulatively during one Stage of the Encounter.
          (Variables|Default|SetScoreValue 0)
          (Default|DebugPrintString (Utilities|String|Append (Utilities|String|BuildString(Integer) (Utilities|String|BuildString(Integer) "" "ELIM POINTS = " _output_get "/") "" (Variables|EncounterLogic|GetEncounterStageScoreTarget) " | ") (Default|EncounterTimersasString)))
          (Default|SetRepParamInt "(TagName=\"RepParam.Integer.Sideways.Encounter.Stage.Score.Eliminated\")" (Variables|EncounterLogic|GetEncounterTotalScoreEliminated)) ; Used for UI.
          (if _overkill
            (bind _output_get_2 (Variables|EncounterLogic|SetEncounterStageOverkillPoints _overkillorunderkillscore))
            (Default|DebugPrintString (Utilities|String|BuildString(Integer) "" "STAGE OVERKILL POINTS = " _output_get_2))
            (Default|SetRepParamInt "(TagName=\"RepParam.Integer.Sideways.Encounter.Total.Overkill.Points\")" (+ (Variables|EncounterLogic|GetEncounterTotalOverkillPoints) (Variables|EncounterLogic|GetEncounterStageOverkillPoints))) ; Used for UI.
          ))))
    (:then_1
      (bind _data (Utilities|FlowControl|MergeExecutionwithData (stale Custom|OnEndPlay_Event _asfort_pawn) _asfort_pawn)
        (:Then
          (EventDispatchers|UnbindEventfromOnDied _data (delegate Custom|OnDied_Event) :!class /Script/FortniteGame.FortPawn)
          (Game|UnbindEventfromOnEndPlay _data (delegate Custom|OnEndPlay_Event))
          (if (not (Variables|EncounterLogic|GetAutoKillInProgress))
            (Utilities|Array|RemoveItem (Variables|EncounterLogic|GetSpawnedPawns) _data))
          (CallFunction|TrytoProgressEncounter) ; See if ALL THE THINGS are dead and whether our Encounter is done.
          (CallFunction|ElimFeedback _data)
          (CallFunction|UpdateTimerUI) ; After an AI is Eliminated
        )))))

(event Custom|OnActorSpawnedResult_Event
  (params (SpawnedActor :type Actor%20Object%20Reference) (bResult :type Boolean))
  (bind _buildingowner (CallFunction|GetBuildingOwner))
  (bind _npc_misc_score_hf (Variables|EncounterSetup|GetNPC%20Misc%20Score%20HF))
  (bind (_asbool _asintrounded) (Default|GetScalableFloatHotfixValue _npc_misc_score_hf))
  (bind _npc_smallbag_score_hf (Variables|EncounterSetup|GetNPC%20SmallBag%20Score%20HF))
  (bind (_asbool_2 _asintrounded_2) (Default|GetScalableFloatHotfixValue _npc_smallbag_score_hf))
  (bind _npc_io_score_hf (Variables|EncounterSetup|GetNPC%20IO%20Score%20HF))
  (bind (_asbool_3 _asintrounded_3) (Default|GetScalableFloatHotfixValue _npc_io_score_hf))
  (bind _spawnedpawns (Variables|EncounterLogic|GetSpawnedPawns))
  ;[1 1 1] AI Spawned - Add to List - Track Score Points
  (if bResult
    (CallFunction|SetEncounterTimer) ; Set Encounter Timer on SPAWN instead of Volume Enabling since there might be a delay.\n\nWill only Set Encounter Timer if it is not yet set.
    (Utilities|FlowControl|Sequence
      (:then_0
        (bind _asfort_aipawn (Utilities|Casting|CastToFortAIPawn SpawnedActor)
          (:then
            (Variables|EncounterLogic|SetEncounterTotalScoreSpawned (+ (FortAIPawn|GetScoreValue _asfort_aipawn) (Variables|EncounterLogic|GetEncounterTotalScoreSpawned)))
            (Variables|EncounterLogic|SetEncounterStageScoreSpawned (+ (FortAIPawn|GetScoreValue _asfort_aipawn) (Variables|EncounterLogic|GetEncounterStageScoreSpawned)))
            (Utilities|Array|AddUnique (Variables|EncounterLogic|GetSpawnedPawns) _asfort_aipawn))
          (:CastFailed
            ;[1 1 1] If NPC Boss is Spawned set the Leash via GameplayEvent
            (bind _asfort_player_pawn_athena (Utilities|Casting|CastToFortPlayerPawnAthena SpawnedActor))
            (Utilities|Casting|CastToFortAthenaAIBotController (Pawn|GetController _asfort_player_pawn_athena))
            (Utilities|IsValid _buildingowner
              (:Is%20Valid
                (Ability|SendGameplayEventtoActor _asfort_player_pawn_athena "(TagName=\"Corruption.Event.Boss.Spawned\")" (Utilities|Struct|MakeGameplayEventData "(TagName=\"Corruption.Event.Boss.Spawned\")" 0 0 _buildingowner (Components|GetOwner self)))
                (bind _asfort_player_pawn_athena_2 (Utilities|Casting|CastToFortPlayerPawnAthena SpawnedActor))
                (Variables|EncounterLogic|SetEncounterTotalScoreSpawned (+ (Math|Integer|SelectInt _asintrounded_3 (Math|Integer|SelectInt _asintrounded_2 _asintrounded (GameplayTags|HasMatchingGameplayTag _asfort_player_pawn_athena_2 "(TagName=\"AISpawnerData.Type.Boss.SmallBag\")")) (GameplayTags|HasMatchingGameplayTag _asfort_player_pawn_athena_2 "(TagName=\"AISpawnerData.Type.IO\")")) (Variables|EncounterLogic|GetEncounterTotalScoreSpawned)))
                (Variables|EncounterLogic|SetEncounterStageScoreSpawned (+ (Math|Integer|SelectInt _asintrounded_3 (Math|Integer|SelectInt _asintrounded_2 _asintrounded (GameplayTags|HasMatchingGameplayTag _asfort_player_pawn_athena_2 "(TagName=\"AISpawnerData.Type.Boss.SmallBag\")")) (GameplayTags|HasMatchingGameplayTag _asfort_player_pawn_athena_2 "(TagName=\"AISpawnerData.Type.IO\")")) (Variables|EncounterLogic|GetEncounterStageScoreSpawned)))
                (Utilities|Array|AddUnique (Variables|EncounterLogic|GetSpawnedPawns) _asfort_player_pawn_athena_2)))))
        ;[1 1 1] AI Died - Remove from List - Try to End or Progress Encounter
        (EventDispatchers|AssignOnDied _spawnedpawns (delegate Custom|OnDied_Event) :!class /Script/FortniteGame.FortPawn)
        (Game|AssignOnEndPlay _spawnedpawns (delegate Custom|OnEndPlay_Event)))
      (:then_1
        (bind _asenemy_pawn_parent_deimos (Utilities|Casting|CastToEnemyPawn_Parent_Deimos SpawnedActor))
        (Variables|/Game/Characters/Enemies/Blueprints/EnemyPawn_Parent_Deimos.EnemyPawn_Parent_Deimos_C|SetGC_OnSpawnFX (Variables|Default|GetGC_SpawnAI) _asenemy_pawn_parent_deimos)))))

(event EventBeginPlay
  ;[1 1 1] Begin Play - Cache any values that the server needs
  (if (Networking|IsServer)
    (bind _gamestate (Class|FortAsyncActionCurrentPlaylistReady|PlaylistDataReadyAsync)
      (:OnReady
        (Variables|EncounterLogic|SetCachedGameState _gamestate)
        (CallFunction|SetEncounterState "NewEnumerator0")
        (Utilities|FlowControl|DelayUntilNextTick) ; Wait for the actor that spawned me \nto register to my events
        (CallFunction|SpawninCosmetics)
        (CallFunction|SetupEncounterDespawn))) ; Make sure we load any hotfix overrides
  ))

(event Custom|SetupEncounterDespawn
  (bind _row_maxidlelife (Variables|MiscLogic|GetRow_MaxIdleLife))
  (bind _returnvalue (ScalableFloat|GetValueatLevel _row_maxidlelife))
  ;[1 1 1] Despawn Logic - Set up a timer and when it goes off if the encounter is in the inactive state despawn it. If the Encounter is outside the safezone when the safe zone updated despawn it.
  (Utilities|FlowControl|Sequence
    (:then_0
      (if (Variables|MiscLogic|GetDoesEncounterEverDespawn)
        (if (> _returnvalue 0.0)
          (bind _returnvalue_2 (Utilities|Time|SetTimerbyEvent (delegate Custom|ServerDespawnTriggered) (select (CallFunction|GameIsinPreDrop) (ScalableFloat|GetValueatLevel (Variables|MiscLogic|GetRow_MaxIdleLifeForPreDrop)) _returnvalue)))
          (Variables|MiscLogic|SetTimer_EncounterInactiveDespawn _returnvalue_2))))
    (:then_1
      (bind _returnvalue_3 (Utilities|Time|SetTimerbyEvent (delegate Custom|CheckInSafezoneTimer) (ScalableFloat|GetValueatLevel (Variables|MiscLogic|GetRow_OutOfSafeZoneFrequencyCheck)) true))
      (Variables|MiscLogic|SetTimer_OutOfSafeZoneDespawn _returnvalue_3))))

(event Custom|SpawnInCosmetics
  ;[1 1 1] Spawn In Cosmetics - Play any cosmetics that need to play only for clients that see the encounter actor spawn in
  (Default|CallSpawnCosmeticEvent (Variables|Spawn/Despawn/DestroyCosmetics|GetGC_Spawn)))

(event Custom|EncounterFinished
  ;[1 1 1] Encounter Finished - Trigger reward and fanfair, then trigger the kill sequence
  (if (Networking|IsServer)
    (CallFunction|SpawnSuccessLoot)
    (Default|CallKillEvent)
    (CallFunction|DestroyCosmetics)))

(event Custom|StartEncounter
  ;[1 1 1] Triggers the Encounter
  (CallFunction|CopyComponentReferences)
  (if (Utilities|Enum|Equal(Enum) (CallFunction|GetEncounterState) "NewEnumerator0")
    (CallFunction|ActivateEncounterBubble)
    (Default|StartLWMEncounter)
    (if (Networking|IsServer)
      (Variables|/SidewaysGameplay/Environmental/SidewaysEncounter/BGA_Athena_SidewaysEncounter_BubbleSmall.BGA_Athena_SidewaysEncounter_BubbleSmall_C|SetBubbleVisible true (CallFunction|GetSidewaysBubble)) ; Unhide the bubble
      (CallFunction|SendInitialDatatoUI))))

(event Custom|ServerDespawnTriggered
  ;[1 1 1] Despawn Logic - Set up a timer and when it goes off if the encounter is in the inactive state despawn it. If the Encounter is outside the safezone when the safe zone updated despawn it.
  (if (Networking|IsServer)
    (if (Utilities|Enum|Equal(Enum) (CallFunction|GetEncounterState) "NewEnumerator0")
      (Variables|/SidewaysGameplay/Environmental/SidewaysEncounter/BGA_Athena_SidewaysEncounter_Grabber.BGA_Athena_SidewaysEncounter_Grabber_C|SetEncounterShutdown true (CallFunction|GetGrabber))
      (CallFunction|DespawnCosmetics)
      (Default|CallKillEvent))))

(event Custom|DespawnCosmetics
  ;[1 1 1] Despawn Cosmetics - Play any cosmetics that need to play only for clients that see the encounter actor spawn out. Spawn out happens if the encounter is never triggered and we want to remove it from the world
  (Default|CallSpawnCosmeticEvent (Variables|Spawn/Despawn/DestroyCosmetics|GetGC_Despawn)))

(event Custom|CheckInSafezoneTimer
  (bind _buildingowner (CallFunction|GetBuildingOwner))
  ;[1 1 1] Despawn Logic - Set up a timer and when it goes off if the encounter is in the inactive state despawn it. If the Encounter is outside the safezone when the safe zone updated despawn it.
  (Utilities|IsValid _buildingowner
    (:Is%20Valid
      (bind _returnvalue (SafeZone|IsLocationinSafeZone (Transformation|GetActorLocation _buildingowner)))
      (if (not _returnvalue)
        (CallFunction|ServerDespawnTriggered)))))

(event Custom|SpawnSuccessLoot
  (bind _buildingowner (CallFunction|GetBuildingOwner))
  (bind _loottier_success (Variables|MiscLogic|GetLootTier_Success))
  ;[1 1 1] Spawn Success Loot - Spawn loot given from succeeding in the Encounter
  (if (Utilities|Name|NotEqual(Name) _loottier_success "None")
    (Utilities|IsValid _buildingowner
      (:Is%20Valid
        (Default|DebugPrintString (Utilities|String|BuildString(Name) "" "LUTE = " _loottier_success))
        (Class|AthenaFunctionLibrary|SpawnLootfromTable _loottier_success 1 -1 (+ (Transformation|GetActorLocation _buildingowner) "0, 0,100.000000") (Transformation|GetActorForwardVector _buildingowner) 0 true true)
        (CallFunction|EQSSpawnGhostRocks)
        (CallFunction|SpawnOverkillLoot)))))

(event Custom|DestroyCosmetics
  ;[1 1 1] Destroy Cosmetics - Play any cosmetics that need to play only for clients that see the encounter actor be destroyed. Destroy happens if the encounter is triggered and "collapses"
  (Default|CallSpawnCosmeticEvent (select (Variables|EncounterLogic|GetFinalWaveBeaten) (Variables|Spawn/Despawn/DestroyCosmetics|GetGC_Closing_Success) (Variables|Spawn/Despawn/DestroyCosmetics|GetGC_Destruction_Failed))))

(event Custom|ActivateEncounterBubble
  ;[1 1 1] Activate Encounter Bubble - Fire off any logic to start the encounter and clear any timers that should only be on while in the inactive state
  (if (Networking|IsServer)
    (CallFunction|SetEncounterState "NewEnumerator1")
    (Utilities|Time|ClearandInvalidateTimerbyHandle (Variables|MiscLogic|GetTimer_OutOfSafeZoneDespawn))
    (Utilities|Time|ClearandInvalidateTimerbyHandle (Variables|MiscLogic|GetTimer_EncounterInactiveDespawn))))

(event Custom|BubbleScalingFinished
  (bind _encounterstate (CallFunction|GetEncounterState))
  ;[1 1 1] Bubble Scaling Finished - (This can be used to trigger FX and also spawning as needed)
  (if (Math|Boolean|ANDBoolean (Utilities|Enum|NotEqual(Enum) _encounterstate "NewEnumerator6") (Utilities|Enum|NotEqual(Enum) _encounterstate "NewEnumerator0") (Utilities|Enum|NotEqual(Enum) _encounterstate "NewEnumerator3"))
    (CallFunction|TrackOverkillTotal) ; Track Bonus Elims for extra loot
    (CallFunction|StartEncounterStage)
    (CallFunction|PlaceBuildingEnvironmentals)
    (elif (Utilities|Enum|Equal(Enum) _encounterstate "NewEnumerator3")
      (CallFunction|TrackOverkillTotal) ; Track Bonus Elims for extra loot
      (CallFunction|EncounterFinished)))
  (CallFunction|RefreshActorsinVolume)
  ;[1 1 1] Set a value on the bubble to make sure it is visible for clients when they enter the range of the actor
  (if (Networking|IsServer)
    (Variables|/SidewaysGameplay/Environmental/SidewaysEncounter/BGA_Athena_SidewaysEncounter_BubbleSmall.BGA_Athena_SidewaysEncounter_BubbleSmall_C|SetBubbleCurrentScale (Variables|MiscLogic|GetBubbleEndScale) (CallFunction|GetSidewaysBubble))))

(event Custom|EncounterStateChanged
  (params (State :type Enum_Athena_SidewaysEncounter_State%20Enum :default "NewEnumerator0"))
  ;[1 1 1] Encoutner State Changed - Update the circle and trigger the events for the different states
  (Default|DebugPrintString (Utilities|String|Append "STATE CHANGE | " (Utilities|String|EnumtoString State)))
  (CallFunction|ScaleBubble)
  (Default|PlayStateChangeGC)
  (Default|UpdateLoopingAudio 0)
  (switch Utilities|FlowControl|Switch|SwitchonEnum_Athena_SidewaysEncounter_State State
    (:NewEnumerator0
      (CallFunction|InactiveStarted))
    (:NewEnumerator1
      (CallFunction|Stage1Started))
    (:NewEnumerator2
      (CallFunction|Stage2Started))
    (:NewEnumerator5
      (CallFunction|Stage3Started))
    (:NewEnumerator3
      (CallFunction|ClosingStarted))))

(event Custom|SpawnOverkillLoot
  (bind _buildingowner (CallFunction|GetBuildingOwner))
  (bind _loottier_overkill (Variables|MiscLogic|GetLootTier_Overkill))
  ;[1 1 1] Spawn Overkill Loot
  (if (> (Variables|EncounterLogic|GetEncounterTotalOverkillPoints) 0)
    (bind _keys (Utilities|Map|Keys (Variables|EncounterSetup|GetOverkillLootTiers_Solo)))
    (bind (_array_element _array_index) (Utilities|Array|ForEachLoopwithBreak _keys :!label _label)
      (:LoopBody
        (if (>= (Variables|EncounterLogic|GetEncounterTotalOverkillPoints) _array_element)
          (bind _values (Utilities|Map|Values (Variables|EncounterSetup|GetOverkillLootTiers_Solo)))
          (Default|DebugPrintString (Utilities|String|BuildString(Name) "" "OVERKILL LOOT TIER = " (Utilities|Array|Get(acopy) _values _array_index)))
          (Variables|MiscLogic|SetLootTier_Overkill (Utilities|Array|Get(acopy) _values _array_index)) ; Best Tier of Overkill reward we have achieved
          (goto _label Break)))
      (:Completed
        (if (Utilities|Name|NotEqual(Name) _loottier_overkill "None")
          (Utilities|IsValid _buildingowner
            (:Is%20Valid
              (Class|AthenaFunctionLibrary|SpawnLootfromTable _loottier_overkill 1 -1 (Transformation|GetActorLocation _buildingowner) (Transformation|GetActorForwardVector _buildingowner) 0 true true))))))))

(event Custom|ScaleBubble
  (bind _encounterstate (CallFunction|GetEncounterState))
  ;[1 1 1] Scale Bubble - Scale the bubble to the new size it needs to be based on the state change.
  (Variables|EncounterLogic|SetEncounterStageActive false) ; Track that we're in a transitional phase.
  (bind _settings (Default|GetEncounterSettings _encounterstate)
    (:Found
      (bind _settings_2 (Default|GetEncounterSettings (Utilities|Enum|BytetoEnumEnum_Athena_SidewaysEncounter_State (Math|Byte|Byte-Byte _encounterstate 1)))
        (:Found
          (Variables|MiscLogic|SetBubbleEndScale (select (CallFunction|EncounterClosing) 0.01 (/ (CallFunction|EnsureVisualRadiusIsinSteppedIncrements (ScalableFloat|GetValueatLevel (Utilities|Struct|BreakStructAthenaSidewaysEncounterState _settings))) (Variables|MiscLogic|GetRadiusScaleMult))))
          (Variables|MiscLogic|SetBubbleStartScale (/ (CallFunction|EnsureVisualRadiusIsinSteppedIncrements (ScalableFloat|GetValueatLevel (Utilities|Struct|BreakStructAthenaSidewaysEncounterState _settings_2))) (Variables|MiscLogic|GetRadiusScaleMult)))
          (Default|DestroyBuildings (CallFunction|EnsureVisualRadiusIsinSteppedIncrements (ScalableFloat|GetValueatLevel (Utilities|Struct|BreakStructAthenaSidewaysEncounterState _settings))))
          (Default|CallScaleBubbleEvent (Variables|MiscLogic|GetBubbleStartScale) (Variables|MiscLogic|GetBubbleEndScale)) ; This relies on an actor listening to this, \nand then calling BubbleScalingFinished
          (if (and (CallFunction|EncounterClosing) (Networking|IsServer))
            (Variables|/SidewaysGameplay/Environmental/SidewaysEncounter/BGA_Athena_SidewaysEncounter_EnvCoreSmall.BGA_Athena_SidewaysEncounter_EnvCoreSmall_C|SetDisableWorldFX true (CallFunction|GetBuildingOwner))))))))

(event EventEndPlay
  (params (EndPlayReason :type EEndPlayReason%20Enum :default "Destroyed"))
  (bind _returnvalue (Game|GetGameState))
  (bind _returnvalue_2 (Actor|GetComponentbyClass _returnvalue "/Script/LagerRuntime.FortAthenaLivingWorldManager"))
  ;[1 1 1] End Play - Clean up LWM Encounter
  (Utilities|FlowControl|Sequence
    (:then_0
      (Utilities|IsValid _returnvalue
        (:Is%20Valid
          (Utilities|IsValid _returnvalue_2
            (:Is%20Valid
              (Class|FortAthenaLivingWorldManager|EndEncounter _returnvalue_2 (Variables|EncounterLogic|GetLWMEncounterInstance)))))))))

(event Custom|EQSSpawnGhostRocks
  (bind _buildingowner (CallFunction|GetBuildingOwner))
  ;[1 1 1] Spawn Spooky Mist - Spawn in the spooky mist consumables using EQS to find the placement locations
  (Utilities|IsValid _buildingowner
    (:Is%20Valid
      (if (> (Math|Float|Floor (ScalableFloat|GetValueatLevel (Variables|EncounterEnvironmentalActors|GetSpookyMistCount))) 0)
        (bind _returnvalue (AI|EQS|RunEQSQuery (Variables|EncounterEnvironmentalActors|GetEQS_SpookyMist) _buildingowner "AllMatching"))
        (AI|EQS|BindEventtoOnQueryFinished _returnvalue (delegate Custom|SpawnGhostRocks))))))

(event Custom|SpawnGhostRocks
  (params (QueryInstance :type EQS%20Query%20Instance%20Object%20Reference) (QueryStatus :type EEnvQueryStatus%20Enum :default "Processing"))
  (bind _spookymistlocations (Variables|EncounterEnvironmentalActors|GetSpookyMistLocations))
  ;[1 1 1] Spawn Spooky Mist - Spawn in the spooky mist consumables using EQS to find the placement locations
  (bind (_resultlocations _returnvalue) (AI|EQS|GetQueryResultsasLocations QueryInstance))
  (if _returnvalue
    (bind _output_get (Variables|EncounterEnvironmentalActors|SetSpookyMistLocations _resultlocations))
    (Utilities|Array|Shuffle _output_get)
    (for _index (range (Math|Float|Floor (ScalableFloat|GetValueatLevel (Variables|EncounterEnvironmentalActors|GetSpookyMistCount))))
      (if (Utilities|Array|IsValidIndex _spookymistlocations _index)
        (Building|SpawnBuildingActor (Variables|EncounterEnvironmentalActors|GetSpookyMist) (Math|Transform|MakeTransform (Math|Vector|vector+vector (Utilities|Array|Get(aref) _spookymistlocations _index) (Math|Vector|MakeVector 0.0 0.0 (Variables|EncounterEnvironmentalActors|GetEnvironmentalsZOffset))) (Math|Rotator|MakeRotator 0.0 0.0 (Math|Random|RandomFloatinRange 0.0 3.6e+02)) "1.000000,1.000000,1.000000"))))))

(event Custom|InactiveStarted)

(event Custom|Stage1Started
  ;[1 1 1] Event States Started -
  (CallFunction|CEUpdateCoreFXStage1)
  (bind _visualradiusinsteppedincrements (Utilities|GetCurrentEncounterVisualRadius)
    (:Found
      (CallFunction|EnableTrackingRadius _visualradiusinsteppedincrements (Variables|Default|GetMapTrackerColor) (CallFunction|GetSidewaysBubble)))))

(event Custom|CE_UpdateCoreFX_Stage1
  ;[1 1 1] Event States Started -
  (Class|BGAAthenaSidewaysEncounterEnvCoreSmall|ChangeCoreFX (CallFunction|GetBuildingOwner) 1))

(event Custom|CE_UpdateCoreFX_Stage2
  ;[1 1 1] Event States Started -
  (Class|BGAAthenaSidewaysEncounterEnvCoreSmall|ChangeCoreFX (CallFunction|GetBuildingOwner) 2))

(event Custom|Stage2Started
  (bind _lwmencounterinstance (Variables|EncounterLogic|GetLWMEncounterInstance))
  ;[1 1 1] Event States Started -
  (Class|BGAAthenaSidewaysEncounterBubbleSmall|EncounterStageCompleted (CallFunction|GetSidewaysBubble))
  (CallFunction|CEUpdateCoreFXStage2)
  (Utilities|IsValid _lwmencounterinstance
    (:Is%20Valid
      (Class|FortAthenaLivingWorldEncounterInstance|AdvanceEncounterStage _lwmencounterinstance))))

(event Custom|Stage3Started
  (bind _lwmencounterinstance (Variables|EncounterLogic|GetLWMEncounterInstance))
  ;[1 1 1] Event States Started -
  (Class|BGAAthenaSidewaysEncounterBubbleSmall|EncounterStageCompleted (CallFunction|GetSidewaysBubble))
  (CallFunction|CEUpdateCoreFXStage3)
  (Utilities|IsValid _lwmencounterinstance
    (:Is%20Valid
      (Class|FortAthenaLivingWorldEncounterInstance|AdvanceEncounterStage _lwmencounterinstance))))

(event Custom|ClosingStarted
  ;[1 1 1] Event States Started -
  (CallFunction|CleanUpEnvironmentals))

(event Custom|CE_UpdateCoreFX_Stage3
  ;[1 1 1] Event States Started -
  (Class|BGAAthenaSidewaysEncounterEnvCoreSmall|ChangeCoreFX (CallFunction|GetBuildingOwner) 3))

(fn AddEnvironmentalActors
  (params (Min :type Scalable%20Float%20Structure%20%28by%20ref%29) (Max :type Scalable%20Float%20Structure%20%28by%20ref%29) (Actor :type Building%20Actor%20Class%20Reference) (RandomActor :type Array%20of%20Building%20Actor%20Class%20References) (PickRandomActor :type Boolean))
  (bind _outitem (Utilities|Array|RandomArrayItem RandomActor))
  (bind _returnvalue (ScalableFloat|GetValueatLevel Max))
  (bind _returnvalue_2 (ScalableFloat|GetValueatLevel Min))
  (bind _returnvalue_3 (Math|Float|Floor _returnvalue))
  (bind _returnvalue_4 (Math|Float|Floor _returnvalue_2))
  (bind _returnvalue_5 (Math|Random|RandomIntegerinRange _returnvalue_4 _returnvalue_3))
  (bind _environmentalactorstospawn (Variables|EncounterEnvironmentalActors|GetEnvironmentalActorsToSpawn))
  ;[1 1 1] Given a min and max value find how many of a given building actor we should spawn and add it to are array of actors to spawn. Use a random varient if desired
  (if (> _returnvalue_5 0)
    (for _ (range 1 (+ _returnvalue_5 1))
      (if PickRandomActor
        (if (Utilities|IsValidClass _outitem)
          (Utilities|Array|Add _environmentalactorstospawn _outitem))
        (elif (Utilities|IsValidClass Actor)
          (Utilities|Array|Add _environmentalactorstospawn Actor))))))

(fn CopyComponentReferences
  (bind _buildingowner (CallFunction|GetBuildingOwner))
  (Utilities|IsValid _buildingowner
    (:Is%20Valid
      (Variables|Default|SetEncounterStageVolume (Variables|/SidewaysGameplay/Environmental/SidewaysEncounter/BGA_Athena_SidewaysEncounter_EnvCoreSmall.BGA_Athena_SidewaysEncounter_EnvCoreSmall_C|GetEncounterStageVolume _buildingowner))
      (Variables|Default|SetBubbleScale (Variables|/SidewaysGameplay/Environmental/SidewaysEncounter/BGA_Athena_SidewaysEncounter_EnvCoreSmall.BGA_Athena_SidewaysEncounter_EnvCoreSmall_C|GetBubbleScale _buildingowner)))
    (:Is%20Not%20Valid
      (Development|PrintString "SidewaysEncounter : Failed to copy component references"))))

(fn DebugPrintString
  (params (String :type String) (ForcePrint :type Boolean))
  (Utilities|SwitchHasAuthority
    (:Authority
      (Development|PrintString (Utilities|String|Append "SIDEWAYS ENCOUNTER DEBUG | " (Utilities|String|EnumtoString (CallFunction|GetEncounterState)) " | " String) (or (Variables|EncounterSetup|GetDebugEncounterObject) ForcePrint) (or (Variables|EncounterSetup|GetDebugEncounterObject) ForcePrint)))))

(fn DestroyBuildings
  (params (EndScale :type Float%20%28double-precision%29))
  (bind _bubble (CallFunction|GetSidewaysBubble))
  (bind _bubblescale (Variables|Default|GetBubbleScale))
  (Utilities|IsValid _bubble
    (:Is%20Valid
      (Utilities|IsValid _bubblescale
        (:Is%20Valid
          (Class|BGAAthenaSidewaysEncounterBubbleSmall|DestroyBuildings _bubble EndScale (/ (Components|Timeline|GetTimelineLength _bubblescale) (Components|Timeline|GetPlayRate _bubblescale))))))))

(fn EncounterClosing
  (return (Utilities|Enum|Equal(Enum) (CallFunction|GetEncounterState) "NewEnumerator3")))

(fn EncounterHasSpawnedAI
  (return (> (Utilities|Array|Length (Variables|EncounterLogic|GetSpawnedPawns)) 0)))

(fn EncounterStageScoreEliminationFallbackTargetHit
  (return (and (>= (Variables|EncounterLogic|GetEncounterStageScoreEliminated) (Variables|EncounterLogic|GetEncounterStageScoreSpawned)) (not (Default|EncounterHasSpawnedAI)))))

(fn EncounterStageScoreEliminationTargetHit
  (return (>= (Variables|EncounterLogic|GetEncounterStageScoreEliminated) (Variables|EncounterLogic|GetEncounterStageScoreTarget)) (- (Variables|EncounterLogic|GetEncounterStageScoreEliminated) (Variables|EncounterLogic|GetEncounterStageScoreTarget))))

(fn EncounterStageScoreEliminationTargetOverkill
  (bind (_elimtargethit _overkillorunderkillscore) (Default|EncounterStageScoreEliminationTargetHit))
  (bind _returnvalue (Utilities|Time|IsTimerActivebyFunctionName 0 "EncounterTimerMin"))
  (bind _returnvalue_2 (and _elimtargethit _returnvalue))
  (Default|DebugPrintString (Utilities|String|BuildString(Boolean) (Utilities|String|BuildString(Boolean) (Utilities|String|BuildString(Boolean) "" "Elim target hit? = " _elimtargethit " | ") "Min Timer still active? = " _returnvalue " | ") "OVERKILL? = " _returnvalue_2))
  (return _returnvalue_2 _overkillorunderkillscore))

(fn EncounterTimersAsString
  (return (Utilities|String|BuildString(Float) (Utilities|String|BuildString(Float) "" "MAX TIME = " (Utilities|Time|GetTimerRemainingTimebyFunctionName 0 "EncounterTimer") " | ") "MIN TIME = " (Utilities|Time|GetTimerRemainingTimebyFunctionName 0 "EncounterTimerMin") " | ")))

(fn EnsureVisualRadiusIsInSteppedIncrements
  (params (VisualRadius :type Float%20%28double-precision%29))
  (bind _returnvalue 5e+02)
  (bind _returnvalue_2 (/ VisualRadius _returnvalue))
  (bind _returnvalue_3 (Math|Float|Truncate _returnvalue_2))
  (bind _returnvalue_4 (Math|Integer|Clamp(Integer) _returnvalue_3 2 20))
  (bind _returnvalue_5 (Math|Conversions|ToFloat(Integer) _returnvalue_4))
  (bind _returnvalue_6 (* _returnvalue_5 _returnvalue))
  (Default|DebugPrintString (Utilities|String|BuildString(Float) (Utilities|String|BuildString(Integer) (Utilities|String|BuildString(Float) (Utilities|String|BuildString(Float) (Utilities|String|BuildString(Float) "" "Incoming R = " VisualRadius " | ") "Increment Val = " _returnvalue " | ") "Num Increments = " _returnvalue_2 " | ") "Clamped Increments = " _returnvalue_4 " | ") "RESULTING DIST = " _returnvalue_6))
  (return _returnvalue_6))

(fn ForciblyAutoKillEnemiesBeforeClosingBubble
  (Variables|EncounterLogic|SetAutoKillInProgress true)
  (for _array_element (Variables|EncounterLogic|GetSpawnedPawns)
    (bind _asaicontroller (Utilities|Casting|CastToAIController (Pawn|GetController _array_element))
      (:then
        (GameplayEffects|ApplyGameplayEffectToTarget :self (Ability|GetAbilitySystemComponent _array_element) :GameplayEffectClass (Variables|EncounterSetup|GetGE_DropNoLoot) :Target (Ability|GetAbilitySystemComponent _array_element) :Level 0.0) ; Add this GE so the AI Boss won't drop their Loot when killed by the Encounter because it closed.
        (bind _returnvalue (AI|GetorCreateAIRuntimeParameters _asaicontroller "/Script/FortniteAI.FortAthenaAIRuntimeParameters_AIBotRespawn"))
        (bind _asfort_athena_airuntime_parameters_aibot_respawn (Utilities|Casting|CastToFortAthenaAIRuntimeParameters_AIBotRespawn _returnvalue)
          (:then
            (Variables|/Script/FortniteAI.FortAthenaAIRuntimeParameters_AIBotRespawn|SetbCanRespawnOnDeath false _asfort_athena_airuntime_parameters_aibot_respawn) ; Stop NPC Pawn type AI from respawning as out-of-the-loop when forcibly cleaned up.\n\nYou've lost the chance to finish them off and get drops/rewards.
            (bind _returnvalue_2 (AI|GetorCreateAIRuntimeParameters _asaicontroller "/Script/FortniteAI.FortAthenaAIRuntimeParameters_Inventory"))
            (bind _asfort_athena_airuntime_parameters_inventory (Utilities|Casting|CastToFortAthenaAIRuntimeParameters_Inventory _returnvalue_2)
              (:then
                (Variables|/Script/FortniteAI.FortAthenaAIRuntimeParameters_Inventory|SetbShouldDropCurrencyOnDeath false _asfort_athena_airuntime_parameters_inventory) ; Setop NPC from dropping Gold Bars
              )
              (:CastFailed
                _)))
          (:CastFailed
            _)))
      (:CastFailed
        _))
    (FortPawn|ForceKill _array_element "(TagName=\"Gameplay.Effect.InstantDeath.Despawn\")" (Pawn|GetController _array_element) _array_element)))

(fn GameIsInPreDrop
  ;[1 1 1] Check if the game is actively in one of the "pre drop" phases
  (return (Math|Boolean|ORBoolean (Utilities|Enum|Equal(Enum) (BattleRoyaleGamePhaseLogic|GetGamePhase (BattleRoyaleGamePhaseLogic|GetBattleRoyaleGamePhaseLogicComponent)) "Setup") (Utilities|Enum|Equal(Enum) (BattleRoyaleGamePhaseLogic|GetGamePhase (BattleRoyaleGamePhaseLogic|GetBattleRoyaleGamePhaseLogicComponent)) "Warmup") (Utilities|Enum|Equal(Enum) (BattleRoyaleGamePhaseLogic|GetGamePhase (BattleRoyaleGamePhaseLogic|GetBattleRoyaleGamePhaseLogicComponent)) "Aircraft"))))

(fn GetBuildingOwner
  (bind _testactor (LocalVariables|GetTestActor))
  (bind _returnvalue (Actor|GetParentActor _testactor))
  (bind _testactor_2 (LocalVariables|GetTestActor))
  (bind _asfort_playspace (Utilities|Casting|CastToFortPlayspace (Components|GetOwner)))
  (LocalVariables|SetTestActor (Variables|/Script/PlayspaceSystem.Playspace|GetBoundGameplayVolume _asfort_playspace)) ; First set our owner as TestActor
  (Utilities|IsValid _testactor_2 :!label _label
    (:Is%20Valid
      (bind _asbga_athena_sideways_encounter_env_core_small (Utilities|Casting|CastToBGA_Athena_SidewaysEncounter_EnvCoreSmall _testactor_2)
        (:then
          (return _asbga_athena_sideways_encounter_env_core_small))
        (:CastFailed
          (Utilities|IsValid _returnvalue
            (:Is%20Valid
              (LocalVariables|SetTestActor _returnvalue) ; If it has a parent actor, set TestActor and try again
              (goto _label))
            (:Is%20Not%20Valid
              _)) ; If not, check if the parent actor is valid
        )) ; Check if the test actor is an EnvCore
    )
    (:Is%20Not%20Valid
      _))
  (return 0))

(fn GetEncounterStageData
  (bind (_encounterwavedurationmax_6_76ac01f54939eb0cc9ea559eed563281 _encounterwavedurationmin_7_66c813034b405960f300dba5f8de2f87 _encounterpointeliminationtarget_8_e6a4e7ea4e52e911ed0182966b43eb19 _encounterloottier_9_b27bf7f0481db78eb5ed359395ccfa9b) (Utilities|Struct|BreakStructAthenaSidewaysEncounterStageData (Utilities|Array|Get(aref) (Variables|EncounterSetup|GetEncounterStageData_Squad) (LocalVariables|GetStageAsInt))))
  (bind (_encounterwavedurationmax_6_76ac01f54939eb0cc9ea559eed563281_2 _encounterwavedurationmin_7_66c813034b405960f300dba5f8de2f87_2 _encounterpointeliminationtarget_8_e6a4e7ea4e52e911ed0182966b43eb19_2 _encounterloottier_9_b27bf7f0481db78eb5ed359395ccfa9b_2) (Utilities|Struct|BreakStructAthenaSidewaysEncounterStageData (Utilities|Array|Get(aref) (Variables|EncounterSetup|GetEncounterStageData_Solo) (LocalVariables|GetStageAsInt))))
  (bind (_encounterwavedurationmax_6_76ac01f54939eb0cc9ea559eed563281_3 _encounterwavedurationmin_7_66c813034b405960f300dba5f8de2f87_3 _encounterpointeliminationtarget_8_e6a4e7ea4e52e911ed0182966b43eb19_3 _encounterloottier_9_b27bf7f0481db78eb5ed359395ccfa9b_3) (Utilities|Struct|BreakStructAthenaSidewaysEncounterStageData (Utilities|Array|Get(aref) (Variables|EncounterSetup|GetEncounterStageData_Duo) (LocalVariables|GetStageAsInt))))
  (bind (_encounterwavedurationmax_6_76ac01f54939eb0cc9ea559eed563281_4 _encounterwavedurationmin_7_66c813034b405960f300dba5f8de2f87_4 _encounterpointeliminationtarget_8_e6a4e7ea4e52e911ed0182966b43eb19_4 _encounterloottier_9_b27bf7f0481db78eb5ed359395ccfa9b_4) (Utilities|Struct|BreakStructAthenaSidewaysEncounterStageData (Utilities|Array|Get(aref) (Variables|EncounterSetup|GetEncounterStageData_Trio) (LocalVariables|GetStageAsInt))))
  (bind (_solo _duo _trio _squad_or_rumble) (Default|GetPlaylistGametype))
  (Utilities|FlowControl|Sequence
    (:then_0
      (switch Utilities|FlowControl|Switch|SwitchonEnum_Athena_SidewaysEncounter_State (CallFunction|GetEncounterState)
        (:NewEnumerator1
          (LocalVariables|SetStageAsInt 0))
        (:NewEnumerator2
          (LocalVariables|SetStageAsInt 1))
        (:NewEnumerator5
          (LocalVariables|SetStageAsInt 2))))
    (:then_1
      (if _solo
        (return _encounterwavedurationmax_6_76ac01f54939eb0cc9ea559eed563281_2 _encounterwavedurationmin_7_66c813034b405960f300dba5f8de2f87_2 _encounterpointeliminationtarget_8_e6a4e7ea4e52e911ed0182966b43eb19_2 _encounterloottier_9_b27bf7f0481db78eb5ed359395ccfa9b_2 (Variables|EncounterSetup|GetOverkillLootTiers_Solo))
        (elif _duo
          (return _encounterwavedurationmax_6_76ac01f54939eb0cc9ea559eed563281_3 _encounterwavedurationmin_7_66c813034b405960f300dba5f8de2f87_3 _encounterpointeliminationtarget_8_e6a4e7ea4e52e911ed0182966b43eb19_3 _encounterloottier_9_b27bf7f0481db78eb5ed359395ccfa9b_3 (Variables|EncounterSetup|GetOverkillLootTiers_Duo))
          (elif _trio
            (return _encounterwavedurationmax_6_76ac01f54939eb0cc9ea559eed563281_4 _encounterwavedurationmin_7_66c813034b405960f300dba5f8de2f87_4 _encounterpointeliminationtarget_8_e6a4e7ea4e52e911ed0182966b43eb19_4 _encounterloottier_9_b27bf7f0481db78eb5ed359395ccfa9b_4 (Variables|EncounterSetup|GetOverkillLootTiers_Trio))
            (elif _squad_or_rumble
              (return _encounterwavedurationmax_6_76ac01f54939eb0cc9ea559eed563281 _encounterwavedurationmin_7_66c813034b405960f300dba5f8de2f87 _encounterpointeliminationtarget_8_e6a4e7ea4e52e911ed0182966b43eb19 _encounterloottier_9_b27bf7f0481db78eb5ed359395ccfa9b (Variables|EncounterSetup|GetOverkillLootTiers_Squad)))))))))

(fn GetEncounterState
  ;[1 1 1] Get the state from the BGA
  (return (Variables|/SidewaysGameplay/Environmental/SidewaysEncounter/BGA_Athena_SidewaysEncounter_EnvCoreSmall.BGA_Athena_SidewaysEncounter_EnvCoreSmall_C|GetEncounterState (CallFunction|GetBuildingOwner))))

(fn GetGrabber
  (bind _buildingowner (CallFunction|GetBuildingOwner))
  (bind _grabber (Variables|/SidewaysGameplay/Environmental/SidewaysEncounter/BGA_Athena_SidewaysEncounter_EnvCoreSmall.BGA_Athena_SidewaysEncounter_EnvCoreSmall_C|GetGrabber _buildingowner))
  (bind _childactor (Variables|/Script/Engine.ChildActorComponent|GetChildActor _grabber))
  (Utilities|IsValid _childactor
    (:Is%20Valid
      (bind _asbga_athena_sideways_encounter_grabber (Utilities|Casting|CastToBGA_Athena_SidewaysEncounter_Grabber _childactor))
      (return _asbga_athena_sideways_encounter_grabber))))

(fn GetPlaylistGametype
  (bind _cachedgamestate (Variables|EncounterLogic|GetCachedGameState)
    (:then
      (return (GameplayTags|HasTag (Class|FortGameStateAthena|GetAthenaPlaylistContextTags _cachedgamestate) "(TagName=\"Athena.Playlist.Solo\")") (GameplayTags|HasTag (Class|FortGameStateAthena|GetAthenaPlaylistContextTags _cachedgamestate) "(TagName=\"Athena.Playlist.Duo\")") (GameplayTags|HasTag (Class|FortGameStateAthena|GetAthenaPlaylistContextTags _cachedgamestate) "(TagName=\"Athena.Playlist.Trios\")") (or (GameplayTags|HasTag (Class|FortGameStateAthena|GetAthenaPlaylistContextTags _cachedgamestate) "(TagName=\"Athena.Playlist.Squad\")") (GameplayTags|HasTag (Class|FortGameStateAthena|GetAthenaPlaylistContextTags _cachedgamestate) "(TagName=\"Athena.Playlist.TeamRumble\")"))))
    (:else)))

(fn GetScalableFloatHotfixValue
  (params (ScalableFloat :type Scalable%20Float%20Structure%20%28by%20ref%29) (AtLevel :type Float%20%28double-precision%29))
  (if (Default|ScalableFloatIsValid ScalableFloat)
    (return (> (Math|Float|Round (ScalableFloat|GetValueatLevel ScalableFloat AtLevel)) 0) (Math|Float|Round (ScalableFloat|GetValueatLevel ScalableFloat AtLevel)) (ScalableFloat|GetValueatLevel ScalableFloat AtLevel))
    (else
      (return (> (Math|Float|Round (Utilities|Struct|BreakScalableFloat ScalableFloat)) 0) (Math|Float|Round (Utilities|Struct|BreakScalableFloat ScalableFloat)) (Utilities|Struct|BreakScalableFloat ScalableFloat)))))

(fn GetSidewaysBubble
  (bind _testactor (LocalVariables|GetTestActor))
  (bind _returnvalue (Actor|GetParentActor _testactor))
  (bind _testactor_2 (LocalVariables|GetTestActor))
  (bind _asfort_playspace (Utilities|Casting|CastToFortPlayspace (Components|GetOwner)))
  (LocalVariables|SetTestActor (Variables|/Script/PlayspaceSystem.Playspace|GetBoundGameplayVolume _asfort_playspace)) ; First set our owner as TestActor
  (Utilities|IsValid _testactor_2 :!label _label
    (:Is%20Valid
      (bind _asbga_athena_sideways_encounter_bubble_small (Utilities|Casting|CastToBGA_Athena_SidewaysEncounter_BubbleSmall _testactor_2)
        (:then
          (return _asbga_athena_sideways_encounter_bubble_small))
        (:CastFailed
          (Utilities|IsValid _returnvalue
            (:Is%20Valid
              (LocalVariables|SetTestActor _returnvalue) ; If it has a parent actor, set TestActor and try again
              (goto _label))
            (:Is%20Not%20Valid
              _)) ; If not, check if the parent actor is valid
        )) ; Check if the test actor is a Bubble
    )
    (:Is%20Not%20Valid
      _))
  (return 0))

(fn PlayStateChangeGC
  ;[1 1 1] Execute the state changed gameplay cue if there is one.
  (bind _settings (Default|GetCurrentEncounterSettings)
    (:Found
      (if (GameplayTags|NotEqual(GameplayTag) :A (.GC_EncounterStateStart_36_97826604412333B00CB55DA41096100E (Utilities|Struct|BreakStructAthenaSidewaysEncounterState _settings)))
        (Default|CallSpawnCosmeticEvent (Utilities|Select :Index (CallFunction|GetEncounterState) :NewEnumerator6 (.GC_EncounterStateStart_36_97826604412333B00CB55DA41096100E (Utilities|Struct|BreakStructAthenaSidewaysEncounterState _settings)) :NewEnumerator0 (.GC_EncounterStateStart_36_97826604412333B00CB55DA41096100E (Utilities|Struct|BreakStructAthenaSidewaysEncounterState _settings)) :NewEnumerator1 (.GC_EncounterStateStart_36_97826604412333B00CB55DA41096100E (Utilities|Struct|BreakStructAthenaSidewaysEncounterState _settings)) :NewEnumerator2 (.GC_EncounterStateStart_36_97826604412333B00CB55DA41096100E (Utilities|Struct|BreakStructAthenaSidewaysEncounterState _settings)) :NewEnumerator5 (.GC_EncounterStateStart_36_97826604412333B00CB55DA41096100E (Utilities|Struct|BreakStructAthenaSidewaysEncounterState _settings)) :NewEnumerator3 (select (Variables|EncounterLogic|GetFinalWaveBeaten) (Variables|Spawn/Despawn/DestroyCosmetics|GetGC_Closing_Success) (Variables|Spawn/Despawn/DestroyCosmetics|GetGC_Closing_Failed))))))))

(fn ProgressState
  (bind _encounterstate (CallFunction|GetEncounterState))
  ;[1 1 1] Increase state on the sever if it is not the final state (triggering the rep notify)
  (if (and (Networking|IsServer) (Utilities|Enum|NotEqual(Enum) _encounterstate "NewEnumerator3"))
    (CallFunction|SetEncounterState (Utilities|Enum|BytetoEnumEnum_Athena_SidewaysEncounter_State (Math|Byte|Byte+Byte _encounterstate 1)))))

(fn ScalableFloatIsValid
  (params (ScalableFloat :type Scalable%20Float%20Structure%20%28by%20ref%29))
  (bind (_value _curve) (Utilities|Struct|BreakScalableFloat ScalableFloat))
  (bind (_curvetable _rowname) (Utilities|Struct|BreakCurveTableRowHandle _curve))
  (return (and (Utilities|IsValid _curvetable :!class /Script/Engine.KismetSystemLibrary) (Utilities|Name|NotEqual(Name) _rowname "None"))))

(fn SendAllDataToUI
  (CallFunction|SendInitialDatatoUI)
  (CallFunction|UpdateTimerUI)
  (Default|SetRepParamInt "(TagName=\"RepParam.Integer.Sideways.Encounter.Stage.Score.Eliminated\")" (Variables|EncounterLogic|GetEncounterTotalScoreEliminated)) ; Used for UI.
  (Default|SetRepParamInt "(TagName=\"RepParam.Integer.Sideways.EncounterState\")" (Math|Conversions|ToInteger(Byte) (CallFunction|GetEncounterState))) ; Used for UI.
)

(fn SendInitialDataToUI
  (bind (_solo _duo _trio _squad_or_rumble) (Default|GetPlaylistGametype))
  (bind _encounterstagedata (LocalVariables|GetEncounterStageData))
  (Utilities|FlowControl|Sequence
    (:then_0
      (if _solo
        (LocalVariables|SetEncounterStageData (Variables|EncounterSetup|GetEncounterStageData_Solo))
        (elif _duo
          (LocalVariables|SetEncounterStageData (Variables|EncounterSetup|GetEncounterStageData_Duo))
          (elif _trio
            (LocalVariables|SetEncounterStageData (Variables|EncounterSetup|GetEncounterStageData_Trio))
            (elif _squad_or_rumble
              (LocalVariables|SetEncounterStageData (Variables|EncounterSetup|GetEncounterStageData_Squad)))))))
    (:then_1
      (for _array_element (Utilities|Array|MakeArray 0 1 2)
        (if (Utilities|Array|IsValidIndex _encounterstagedata _array_element)
          (switch int _array_element
            (:0
              (Default|SetRepParamInt "(TagName=\"RepParam.Integer.Sideways.Encounter.Stage.Score.Target.1\")" (.AsIntRounded (Default|GetScalableFloatHotfixValue (.EncounterPointEliminationTarget_8_E6A4E7EA4E52E911ED0182966B43EB19 (Utilities|Struct|BreakStructAthenaSidewaysEncounterStageData (Utilities|Array|Get(aref) _encounterstagedata _array_element)))))) ; Used for UI.
            )
            (:1
              (Default|SetRepParamInt "(TagName=\"RepParam.Integer.Sideways.Encounter.Stage.Score.Target.2\")" (.AsIntRounded (Default|GetScalableFloatHotfixValue (.EncounterPointEliminationTarget_8_E6A4E7EA4E52E911ED0182966B43EB19 (Utilities|Struct|BreakStructAthenaSidewaysEncounterStageData (Utilities|Array|Get(aref) _encounterstagedata _array_element)))))) ; Used for UI.
            )
            (:2
              (Default|SetRepParamInt "(TagName=\"RepParam.Integer.Sideways.Encounter.Stage.Score.Target.3\")" (.AsIntRounded (Default|GetScalableFloatHotfixValue (.EncounterPointEliminationTarget_8_E6A4E7EA4E52E911ED0182966B43EB19 (Utilities|Struct|BreakStructAthenaSidewaysEncounterStageData (Utilities|Array|Get(aref) _encounterstagedata _array_element)))))) ; Used for UI.
            )))))))

(fn SetEncounterState
  (params (EncounterState :type Enum_Athena_SidewaysEncounter_State%20Enum :default "NewEnumerator6"))
  (bind _buildingowner (CallFunction|GetBuildingOwner))
  ;[1 1 1] Set the encounter state on the BGA and update the UI and any other server things via the state change.
  (if (Networking|IsServer)
    (Utilities|IsValid _buildingowner
      (:Is%20Valid
        (Variables|/SidewaysGameplay/Environmental/SidewaysEncounter/BGA_Athena_SidewaysEncounter_EnvCoreSmall.BGA_Athena_SidewaysEncounter_EnvCoreSmall_C|SetEncounterState EncounterState _buildingowner)
        (Default|SetRepParamInt "(TagName=\"RepParam.Integer.Sideways.EncounterState\")" (Math|Conversions|ToInteger(Byte) EncounterState)) ; Used for UI.
        (CallFunction|EncounterStateChanged EncounterState)))))

(fn SetEncounterSuccessVariables
  (Utilities|FlowControl|Sequence
    (:then_0
      (Variables|MiscLogic|SetLootTier_Success (Variables|EncounterLogic|GetCurrentEncounterLootTier) :!pinned true) ; Update this variable as our current best Loot Tier\n\nThat way we can, say, pass Stage 1, and fail Stage 2, \nbut still get Stage 1's reward.
      (if (Utilities|Enum|Equal(Enum) (CallFunction|GetEncounterState) (Variables|EncounterLogic|GetHighestCombatState))
        (Variables|EncounterLogic|SetFinalWaveBeaten true)))
    (:then_1
      (bind _visualradiusinsteppedincrements (Utilities|GetCurrentEncounterVisualRadius :!pinned true)
        (:Found
          (bind _outfortplayerpawns (Fortnite|GetFortPlayerPawnsInRange (Transformation|GetActorLocation (CallFunction|GetSidewaysBubble)) _visualradiusinsteppedincrements :!pinned true)) ; Get players in Bubble Radius
          (for _array_element _outfortplayerpawns
            (Utilities|IsValid _array_element
              (:Is%20Valid
                (Utilities|IsValid (Ability|GetAbilitySystemComponent _array_element)
                  (:Is%20Valid
                    (GameplayEffects|ApplyGameplayEffectToTarget :self (Ability|GetAbilitySystemComponent _array_element) :GameplayEffectClass (Utilities|Select :Index (CallFunction|GetEncounterState) :NewEnumerator1 (Variables|Quests|GetGE_Stage1_Quest) :NewEnumerator2 (Variables|Quests|GetGE_Stage2_Quest) :NewEnumerator5 (Variables|Quests|GetGE_Stage3_Quest) :NewEnumerator3 (Variables|Default|GetVisited%20Sideways)) :Target (Ability|GetAbilitySystemComponent _array_element) :Level 0.0 :!pinned true) ; Apply GEs with tags for Quests to track participation
                  ))))))) ; Get Bubble Radius\n\nTODO, do we need extra buffer?\nWe already check the in-sideways tags for the GE application.
    )))

(fn SetEQSParamsIntoEncounterPointProvider
  (bind _encountereqsdistancedata (Variables|EncounterSetup|GetEncounterEQSDistanceData))
  (bind _settings (Default|GetEncounterSettings (CallFunction|GetEncounterState))
    (:Found
      (bind _keys (Utilities|Map|Keys (Utilities|Struct|BreakStructAthenaSidewaysEncounterEQSData (Utilities|Map|Find _encountereqsdistancedata (CallFunction|EnsureVisualRadiusIsinSteppedIncrements (ScalableFloat|GetValueatLevel (Utilities|Struct|BreakStructAthenaSidewaysEncounterState _settings))))))) ; Each Key in this Struct is an EQS Param 
      (for _array_element _keys
        (bind _currentlivingworldencounterpointprovidervolume (Variables|EncounterLogic|GetCurrentLivingWorldEncounterPointProviderVolume)
          (:then
            (Class|FortAthenaLivingWorldVolume|SetEQSQueryConfigParam _currentlivingworldencounterpointprovidervolume _array_element (Utilities|Map|Find (Utilities|Struct|BreakStructAthenaSidewaysEncounterEQSData (Utilities|Map|Find _encountereqsdistancedata (CallFunction|EnsureVisualRadiusIsinSteppedIncrements (ScalableFloat|GetValueatLevel (Utilities|Struct|BreakStructAthenaSidewaysEncounterState _settings))))) _array_element)) ; Feed-In Param and Value to EQS
            (Default|DebugPrintString (Utilities|String|BuildString(Float) (Utilities|String|BuildString(Name) (Utilities|String|BuildString(Float) "" "ENCOUNTER RADIUS = " (CallFunction|EnsureVisualRadiusIsinSteppedIncrements (ScalableFloat|GetValueatLevel (Utilities|Struct|BreakStructAthenaSidewaysEncounterState _settings))) " | ") "PARAM = " _array_element " | ") "VAL = " (Utilities|Map|Find (Utilities|Struct|BreakStructAthenaSidewaysEncounterEQSData (Utilities|Map|Find _encountereqsdistancedata (CallFunction|EnsureVisualRadiusIsinSteppedIncrements (ScalableFloat|GetValueatLevel (Utilities|Struct|BreakStructAthenaSidewaysEncounterState _settings))))) _array_element))))
          (:else))))))

(fn SpawnChest
  (params (SpawnTransform :type Transform) (CoreReference :type Building%20Gameplay%20Actor%20Object%20Reference))
  ;[1 1 1] Spawn Chest - We need to do it this way to init exposed on spawn values on the chest.
  (bind (_returnvalue _tracesuccess) (CallFunction|TraceForChestPlacement SpawnTransform))
  (if _tracesuccess
    (bind _returnvalue_2 (Game|SpawnActorfromClass "/SidewaysGameplay/Environmental/SidewaysChests/B_Chest_Athena_SidewaysChest_Super.B_Chest_Athena_SidewaysChest_Super_C" _returnvalue "Undefined" "MultiplyWithRoot" 0 true CoreReference 0 0 0))
    (return _returnvalue_2)
    (else
      (return 0))))

(fn SpawnPillar
  (params (SpawnTransform :type Transform) (CoreReference :type Building%20Gameplay%20Actor%20Object%20Reference))
  ;[1 1 1] Spawn A Random Pillar - We need to do it this way to init exposed on spawn values on the pillar.
  (switch int (Math|Random|RandomIntegerinRange 0 2)
    (:Default
      _)
    (:0
      (bind _returnvalue (Game|SpawnActorfromClass "/SidewaysGameplay/Environmental/SidewaysGeo/Rocks/Prop_SidewaysGeo_Rock_01.Prop_SidewaysGeo_Rock_01_C" SpawnTransform "Undefined" "MultiplyWithRoot" 0 true CoreReference 0 0 0))
      (return _returnvalue))
    (:1
      (bind _returnvalue_2 (Game|SpawnActorfromClass "/SidewaysGameplay/Environmental/SidewaysGeo/Rocks/Prop_SidewaysGeo_Rock_02.Prop_SidewaysGeo_Rock_02_C" SpawnTransform "Undefined" "MultiplyWithRoot" 0 true CoreReference 0 0 0))
      (return _returnvalue_2))
    (:2
      _))
  (bind _returnvalue_3 (Game|SpawnActorfromClass "/SidewaysGameplay/Environmental/SidewaysGeo/Rocks/Prop_SidewaysGeo_Rock_03.Prop_SidewaysGeo_Rock_03_C" SpawnTransform "Undefined" "MultiplyWithRoot" 0 true CoreReference 0 0 0))
  (return _returnvalue_3))

(fn StartLWMEncounter
  (bind (_solo _duo _trio _squad_or_rumble) (Default|GetPlaylistGametype))
  (bind _returnvalue (Game|GetGameState))
  (bind _returnvalue_2 (Actor|GetComponentbyClass _returnvalue "/Script/LagerRuntime.FortAthenaLivingWorldManager"))
  (Utilities|IsValid _returnvalue
    (:Is%20Valid
      (Utilities|IsValid _returnvalue_2
        (:Is%20Valid
          (bind _returnvalue_3 (Class|FortAthenaLivingWorldManager|StartEncounter _returnvalue_2 (select (Utilities|IsValidSoftObjectReference (Variables|EncounterSetup|GetDeimosEncounterForcedTest)) (Variables|EncounterSetup|GetDeimosEncounterForcedTest) (Utilities|Array|RandomArrayItem (select _solo (Variables|EncounterSetup|GetDeimosEncounterTypes_Solo) (select _duo (Variables|EncounterSetup|GetDeimosEncounterTypes_Duo) (select _trio (Variables|EncounterSetup|GetDeimosEncounterTypes_Trio) (select _squad_or_rumble (Variables|EncounterSetup|GetDeimosEncounterTypes_Squad) (Variables|EncounterSetup|GetDeimosEncounterTypes_Solo))))))) (Utilities|Array|MakeArray (Utilities|Casting|CastToFortAthenaLivingWorldPointProviderInterface (Variables|/Script/Engine.ChildActorComponent|GetChildActor (Variables|Default|GetEncounterStageVolume)))) (CallFunction|GetBuildingOwner) true))
          (Variables|EncounterLogic|SetLWMEncounterInstance _returnvalue_3))))))

(fn TraceforChestPlacement
  (params (TransformIn :type Transform))
  ;[1 1 1] Trace for the location of the chest as it is not as forgiving as the rocks. After find the cross product and spin it around the hit normal to get a random rotation on that surface
  (bind (_outhit _returnvalue) (Collision|LineTraceByChannel (+ (.location TransformIn) "0, 0,1000.000000") (+ (.location TransformIn) "0, 0,-1000.000000") "TraceTypeQuery7"))
  (if _returnvalue
    (return (Math|Transform|MakeTransform (.Location (Collision|BreakHitResult _outhit)) (Math|Conversions|RotationFromXVector (Math|Vector|RotateVectorAroundAxis (Math|Vector|CrossProduct (.Normal (Collision|BreakHitResult _outhit)) "0.000000,1.000000, 0") (Math|Random|RandomFloatinRange 0.0 359.0) (.Normal (Collision|BreakHitResult _outhit)))) (.scale TransformIn)) true)
    (else
      (return 0 false))))

(fn UpdateLoopingAudio
  (params (OverrideOrOverkillMusic :type Sound%20Base%20Object%20Reference))
  (bind _buildingowner (CallFunction|GetBuildingOwner))
  ;[1 1 1] Set the audio to the current audio that should be being used
  (Utilities|IsValid OverrideOrOverkillMusic
    (:Is%20Valid
      _)
    (:Is%20Not%20Valid
      (bind _settings (Default|GetCurrentEncounterSettings)
        (:Found
          (goto _label Option%20B))
        (:NotFound
          (Development|PrintString "SidewaysEncounter: No State found to select LOOPING audio")))))
  (bind _data (Utilities|FlowControl|MergeExecutionwithData OverrideOrOverkillMusic (.EncounterLoopingAudio_8_3AA8ED3F4879A23432761F90E5473930 (Utilities|Struct|BreakStructAthenaSidewaysEncounterState _settings)) :!label _label)
    (:Then
      (if (Networking|IsServer)
        (Utilities|IsValid _buildingowner
          (:Is%20Valid
            (Variables|/SidewaysGameplay/Environmental/SidewaysEncounter/BGA_Athena_SidewaysEncounter_EnvCoreSmall.BGA_Athena_SidewaysEncounter_EnvCoreSmall_C|SetMusic _data _buildingowner)))))))

(fn UpdateTimerUI
  (Default|SetRepParamFloat "(TagName=\"RepParam.Float.Sideways.WaveDuration.MaxRemaining\")" (Utilities|Time|GetTimerRemainingTimebyFunctionName 0 "EncounterTimer")) ; Used for UI.
  (Default|SetRepParamFloat "(TagName=\"RepParam.Float.Sideways.WaveDuration.MinRemaining\")" (Utilities|Time|GetTimerRemainingTimebyFunctionName 0 "EncounterTimerMin")) ; Used for UI.
  (Default|SetRepParamFloat "(TagName=\"RepParam.Float.Sideways.WarningCountdownDuration.MaxRemaining\")" (Utilities|Time|GetTimerRemainingTimebyFunctionName 0 "EncounterTimerCountdown")) ; Used for UI.
  (Default|SetRepParamFloat "(TagName=\"RepParam.Float.Sideways.CleanupTimeoutDuration.MaxRemaining\")" (Utilities|Time|GetTimerRemainingTimebyFunctionName 0 "CleanupTimeout")) ; Used for UI.
)
